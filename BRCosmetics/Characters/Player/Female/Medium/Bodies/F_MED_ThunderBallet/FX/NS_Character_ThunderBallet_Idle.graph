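# Idle Niagara effect for the ThunderBallet character cosmetic, drawing electric arcs, beams and a glow on the skeletal mesh.
# System timelines drive lightning spawn rates and sweep bone/socket index, offset and color, with SkinWeightCheck swapping index sets.
# Mesh-based ArcLightning and mesh-plus-sprite BeamLightning feed animated material parameters; Glow001 is a socket-following sprite with IsReactive as spawn probability.
# VariantIndex selects among six palettes, FrontEndScale adjusts lightning scale, and bIsHeadHidden zeroes alpha.
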NiagaraSystem /BRCosmetics/Characters/Player/Female/Medium/Bodies/F_MED_ThunderBallet/FX/NS_Character_ThunderBallet_Idle
dependencies {
 usedDataInterfaces [
  &NiagaraDataInterfaceCurve
  &NiagaraDataInterfaceColorCurve
  &NiagaraDataInterfaceArrayFloat
  &NiagaraDataInterfaceArrayColor
  &NiagaraDataInterfaceSkeletalMesh
 ]
 usedDynamicInputs [
  &SetBoolByFloatComparison
  &Boolean_AndOperation
  &SetBoolByIntComparison
 ]
 usedModules [
  &self:SystemSpawnScript.SystemScriptSource.SystemScriptGraph.NiagaraNodeAssignment_0.SetVariables_#1
  &self:SystemSpawnScript.SystemScriptSource.SystemScriptGraph.NiagaraNodeAssignment_1.SetVariables_#2
  &SystemState
  /Game/Effects/Niagara/Modules/Logic/UpdateValueStateChange
  &Timeline
 ]
}
emitters {
 ArcLightning {
  emitterUpdateScript [
   &EmitterState()
   &SpawnRate(
    SpawnRate=&FloatFromCurve(
     FloatCurve=DataInterface() {propertyValues {Curve {keys [
         {arriveTangent=6.637, interpMode=RCIM_Cubic, leaveTangent=6.637, tangentMode=RCTM_User, time=0, value=0}
         {arriveTangent=16.8545, leaveTangent=-0.0176, time=0.05, value=1}
         <at 0.6 1>
         {arriveTangent=-3, leaveTangent=-3, time=0.65, value=0.25}
         {arriveTangent=-0.7143, interpMode=RCIM_Cubic, leaveTangent=-0.7143, time=0.9, value=0.25}
         {arriveTangent=-2.5182, tangentMode=RCTM_User, time=1, value=0}
        ]}}}
     CurveIndex=$System.Timeline.NormalizedTime
     Scale Curve=15
    )
    Use Spawn Probability=false
   )
  ]
  particleSpawnScript [
   /SetVariables(
    Particles.MeshIndex=&RandomRangeInt(Minimum=0, Maximum=5)
    Particles.bFlipMesh:NiagaraBool=&RandomBool()
   ) {isSetParameters=true}
   &InitializeParticle(
    Color=&MergeHelper_ColorSwitch(
     0=&ColorFromCurve(
      ColorCurve=DataInterface() {propertyValues {
        AlphaCurve {keys [<at 0 1 RCIM_Cubic>,<at 1 1 RCIM_Cubic>]}
        BlueCurve {keys [<at 0.8 2 RCIM_Cubic>,<at 1 1 RCIM_Cubic>]}
        GreenCurve {keys [<at 0.8 0 RCIM_Cubic>,<at 1 0.8914 RCIM_Cubic>]}
        RedCurve {keys [<at 0.8 0.2667 RCIM_Cubic>,<at 1 0.0799 RCIM_Cubic>]}
       }}
      CurveIndex=$System.Timeline001.NormalizedTime
     )
     1=&ColorFromCurve(
      ColorCurve=DataInterface() {propertyValues {
        AlphaCurve {keys [<at 0 1 RCIM_Cubic>,<at 1 1 RCIM_Cubic>]}
        BlueCurve {keys [<at 0 0.1812 RCIM_Cubic>,<at 1 0.2894 RCIM_Cubic>]}
        GreenCurve {keys [<at 0 0.2462 RCIM_Cubic>,<at 1 0.4244 RCIM_Cubic>]}
        RedCurve {keys [<at 0 0.9473 RCIM_Cubic>,<at 1 0.9647 RCIM_Cubic>]}
       }}
      CurveIndex=$System.Timeline001.NormalizedTime
     )
     2=&ColorFromCurve(
      ColorCurve=DataInterface() {propertyValues {
        AlphaCurve {keys [<at 0 1 RCIM_Cubic>,<at 1 1 RCIM_Cubic>]}
        BlueCurve {keys [<at 0 0.0242>, {arriveTangent=0.1603, interpMode=RCIM_Cubic, leaveTangent=0.1603, time=0.25, value=0.0375}, <at 1 0.1845 RCIM_Cubic>]}
        GreenCurve {keys [<at 0 0.0176>, {arriveTangent=0.422, interpMode=RCIM_Cubic, leaveTangent=0.422, time=0.25, value=0.0849}, <at 1 0.4397 RCIM_Cubic>]}
        RedCurve {keys [<at 0 0.0232>, {arriveTangent=0.6564, interpMode=RCIM_Cubic, leaveTangent=0.6564, time=0.25, value=0.15}, <at 1 0.6795 RCIM_Cubic>]}
       }}
      CurveIndex=$System.Timeline001.NormalizedTime
     )
     3=&ColorFromCurve(
      ColorCurve=DataInterface() {propertyValues {
        AlphaCurve {keys [<at 0 1 RCIM_Cubic>,<at 1 1 RCIM_Cubic>]}
        BlueCurve {keys [<at 0 0.5647 RCIM_Cubic>,<at 1 0.7379 RCIM_Cubic>]}
        GreenCurve {keys [<at 0 0.0737 RCIM_Cubic>,<at 1 0.5843 RCIM_Cubic>]}
        RedCurve {keys [<at 0 0.5434 RCIM_Cubic>,<at 1 0.0834 RCIM_Cubic>]}
       }}
      CurveIndex=$System.Timeline001.NormalizedTime
     )
     4=&ColorFromCurve(
      ColorCurve=DataInterface() {propertyValues {
        AlphaCurve {keys [<at 0 1 RCIM_Cubic>,<at 1 1 RCIM_Cubic>]}
        BlueCurve {keys [<at 0.006 0.0802 RCIM_Cubic>,<at 0.25 0.4125>,<at 1 0.807 RCIM_Cubic>]}
        GreenCurve {keys [<at 0.006 0.0048 RCIM_Cubic>,<at 0.25 0.0999>,<at 1 0.4447 RCIM_Cubic>]}
        RedCurve {keys [<at 0.006 0.0015 RCIM_Cubic>,<at 0.25 0.0232>,<at 1 0.0807 RCIM_Cubic>]}
       }}
      CurveIndex=$System.Timeline001.NormalizedTime
     )
     5=&ColorFromCurve(
      ColorCurve=DataInterface() {propertyValues {
        AlphaCurve {keys [<at 0 1 RCIM_Cubic>,<at 1 1 RCIM_Cubic>]}
        BlueCurve {keys [<at 0 0.5271 RCIM_Cubic>,<at 1 0.6654 RCIM_Cubic>]}
        GreenCurve {keys [<at 0 0.0844 RCIM_Cubic>,<at 1 0.2961 RCIM_Cubic>]}
        RedCurve {keys [<at 0 0.3185 RCIM_Cubic>,<at 1 0.5583 RCIM_Cubic>]}
       }}
      CurveIndex=$System.Timeline001.NormalizedTime
     )
     Index=&MakeIntFromFloat(Float=$User.VariantIndex)
     Num=::NewEnumerator4
    )
    Lifetime=&RandomRangeFloat(Minimum=0.3, Maximum=0.5)
    Mesh Scale=&Multiply_Vector(
     A=&Multiply_Vector(
      A=&Multiply_VectorByFloat(
       Vector=&RandomRangeVector(Minimum=<0.4 1 0.5>, Maximum=<0.8 1.4 1>)
       Float=&RandomRangeFloat(Minimum=1, Maximum=1)
      )
      B=&MakeCustomVectorFromBool(Bool=$Particles.bFlipMesh, True Vector=<-1 -1 1>, False Vector=<1 -1 1>)
     )
     B=&VectorFromFloat(Value=&FN_FrontEndFloatModifier(Gameplay=1, Front End=$User.FrontEndScale))
    )
    Mesh Scale Mode=::NewEnumerator4
    UsePositionOffset=false
    Write Material Random=true
   )
   &SkeletalMeshLocation(
    Bone / Socket Index=&MakeCustomIntFromBool(
     Bool=$User.SkinWeightCheck
     True INT=&MakeIntFromFloat(Float=&FloatFromCurve(FloatCurve=DataInterface() {propertyValues {Curve {keys [<at 0 6>,<at 0.4 11>]}}}, CurveIndex=$System.Timeline001.NormalizedTime))
     False INT=&MakeIntFromFloat(Float=&FloatFromCurve(
       FloatCurve=DataInterface() {propertyValues {Curve {keys [<at 0 0>,<at 0.4 1>]}}}
       CurveIndex=$System.Timeline001.NormalizedTime
       Scale Curve=&RandomRangeFloat(Minimum=4, Maximum=4)
      ))
    )
    MaskSpawnGroup=false
    Mesh Sampling Type=::NewEnumerator5
    Sampled Position Offset=&Lerp_Vector(
     A=&RandomRangeVector(Minimum=<25 -7.5 10>, Maximum=<25 7.5 10>)
     B=&RandomRangeVector(Minimum=<0 -10 1>, Maximum=<0 10 2>)
     Alpha=&FloatFromCurve(
      FloatCurve=DataInterface() {propertyValues {Curve {keys [
          {arriveTangent=0.2778, interpMode=RCIM_Cubic, leaveTangent=0.2778, tangentMode=RCTM_User, time=0.2, value=0}
          {arriveTangent=0.6216, interpMode=RCIM_Cubic, leaveTangent=0.2299, tangentMode=RCTM_Break, time=0.4, value=1.5}
          <at 0.45 1>
         ]}}}
      CurveIndex=$System.Timeline001.NormalizedTime
     )
    )
    Skeletal Mesh=$System.Skeletal Mesh
    Socket Sampling Mode=::NewEnumerator5
   ) {module=SkeletalMeshLocation001}
   &InitialMeshOrientation(Mesh Orientation Mode=::NewEnumerator2, Rotation=<0.25 0 0.25>, Use Rotation Vector=true) {module=InitialMeshOrientation001}
   &InitialMeshOrientation(
    Mesh Orientation Mode=::NewEnumerator2
    Rotation=&Lerp_Vector(
     A=<0 0 0>
     B=&RandomRangeVector(Minimum=<0 0 0>, Maximum=<0 0 1>)
     Alpha=&FloatFromCurve(
      FloatCurve=DataInterface() {propertyValues {Curve {
         defaultValue=1
         keys [{interpMode=RCIM_Cubic, tangentMode=RCTM_User, time=0.9, value=0}, {interpMode=RCIM_Cubic, tangentMode=RCTM_User, time=1, value=1}]
        }}}
      CurveIndex=$System.Timeline001.NormalizedTime
     )
    )
    Use Rotation Vector=true
   )
   &CameraOffset(Camera Offset Amount=2) {module=CameraOffset001}
  ]
  particleUpdateScript [
   &ParticleState()
   &DynamicMaterialParameters(
    Index 0 Param 2=&RandomRangeFloat(Minimum=0, Maximum=2)
    Index 0 Param 3=&RandomRangeFloat(Minimum=0, Maximum=1)
    Index 0 Param 4=&MakeCustomFloatFromBool(
     Bool=false
     True Float=1
     False Float=&SelectFloatFromArray(
      Array Sampling Mode=::NewEnumerator0
      Direct Array Index=?
      Float Selection Array=DataInterface() {propertyValues {FloatData [3,1,2,1]}}
      Random Seed=0
     )
    )
    Index 1 Param 1=1
    Index 1 Param 2=2
    Index 1 Param 3=&FloatFromCurve(FloatCurve=DataInterface() {propertyValues {Curve {keys [{interpMode=RCIM_Cubic, tangentMode=RCTM_Break, time=0.5, value=-0.5}, <at 1 0.1>]}}})
    Index 2 Param 1=&Multiply_Float(
     A=&FloatFromCurve(
      FloatCurve=DataInterface() {propertyValues {Curve {keys [{interpMode=RCIM_Cubic, leaveTangent=8.1085, tangentMode=RCTM_Break, time=0.8568, value=1}, {tangentMode=RCTM_User, time=1, value=0}]}}}
      Scale Curve=0.25
     )
     B=&Divide_Float(
      A=&MakeFloatFromVector(Vector=$Particles.Scale, Channel=::NewEnumerator0)
      B=&RandomRangeFloat(Minimum=15, Maximum=20)
     )
    )
    Index 2 Param 2=&FloatFromCurve(
     FloatCurve=DataInterface() {propertyValues {Curve {keys [
         {interpMode=RCIM_Cubic, tangentMode=RCTM_Break, time=0.7, value=0}
         {arriveTangent=-0.0859, interpMode=RCIM_Cubic, tangentMode=RCTM_Break, time=1, value=-1}
        ]}}}
     Scale Curve=0
    )
    Index 2 Param 3=&Waveform(
     Global Input Position=&RandomRangeFloat(Minimum=0, Maximum=15)
     Global Phase=$Particles.Age
     "Waveform [1]"=::NewEnumerator12
     "[1] Phase Scale"=2
     "[1] Frequency"=1
     "[1] Amplitude Min/Max"=<0 1>
    )
    Index 2 Param 4=25
    Index 3 Param 1=&FloatFromCurve(
     FloatCurve=DataInterface() {propertyValues {Curve {keys [
         {interpMode=RCIM_Cubic, leaveTangent=-0.0025, tangentMode=RCTM_Break, time=0, value=0}
         {arriveTangent=1.8528, interpMode=RCIM_Cubic, tangentMode=RCTM_Break, time=1, value=1}
        ]}}}
     Scale Curve=&RandomRangeFloat(Minimum=15, Maximum=20)
    )
    Index 3 Param 3=&FloatFromCurve(FloatCurve=DataInterface() {propertyValues {Curve {keys [
         {interpMode=RCIM_Cubic, leaveTangent=1.1034, tangentMode=RCTM_Break, time=0, value=-1}
         {arriveTangent=2.2399, tangentMode=RCTM_User, time=0.5, value=0.125}
        ]}}})
    Index 3 Param 4=&FloatFromCurve(FloatCurve=DataInterface() {propertyValues {Curve {keys [
         {interpMode=RCIM_Cubic, leaveTangent=3.0601, tangentMode=RCTM_Break, time=0.5, value=-0.5}
         {arriveTangent=3.2053, tangentMode=RCTM_User, time=1, value=1}
        ]}}})
    Module.Index0Param0Write=true
    Module.Index0Param1Write=true
    Module.Index0Param2Write=true
    Module.Index0Param3Write=true
    Module.Index1Param0Write=true
    Module.Index1Param1Write=true
    Module.Index1Param2Write=true
    Module.Index1Param3Write=true
    Module.Index2Param0Write=true
    Module.Index2Param1Write=true
    Module.Index2Param2Write=true
    Module.Index2Param3Write=true
    Module.Index3Param0Write=true
    Module.Index3Param1Write=true
    Module.Index3Param2Write=true
    Module.Index3Param3Write=true
    Write Parameter Index 1=true
    Write Parameter Index 2=true
    Write Parameter Index 3=true
   )
   &SolveForcesAndVelocity(Clamp Velocity=false, Limit Acceleration=false)
   &ScaleColor(
    Scale Alpha=&FloatFromCurve(FloatCurve=DataInterface() {propertyValues {Curve {keys [
         {tangentMode=RCTM_User, time=0, value=1}
         {arriveTangent=-0.0405, interpMode=RCIM_Cubic, leaveTangent=-0.0405, tangentMode=RCTM_User, time=0.7009, value=1}
         {arriveTangent=-8.0089, interpMode=RCIM_Cubic, leaveTangent=-8.0089, tangentMode=RCTM_User, time=1, value=0}
        ]}}})
    Scale RGB=&Lerp_Vector(
     A=&VectorFromFloat(Value=&FloatFromCurve(FloatCurve=DataInterface() {propertyValues {Curve {keys [<at 0.2771 1>,<at 0.3786 1>]}}}, Scale Curve=20))
     B=&VectorFromFloat(Value=&RandomRangeFloat(Minimum=0.5, Maximum=1.5, Evaluation Type=::NewEnumerator1))
     Alpha=&FloatFromCurve(FloatCurve=DataInterface() {propertyValues {Curve {keys [
          {interpMode=RCIM_Cubic, leaveTangent=2.5294, tangentMode=RCTM_Break, time=0, value=0}
          {arriveTangent=-0.2166, interpMode=RCIM_Cubic, tangentMode=RCTM_Break, time=0.4, value=1}
         ]}}})
    )
    ScaleA=true
    ScaleRGB=true
   )
   &ScaleColor(
    Scale Alpha=&MakeCustomFloatFromBool(Bool=$User.bIsHeadHidden, True Float=0, False Float=1)
    ScaleA=true
    ScaleRGB=false
   ) {module=ScaleColor001}
  ]
  rendererClasses [&NiagaraMeshRendererProperties]
  renderers [
   {rendererClass=&NiagaraMeshRendererProperties, rendererIndex=0}
   {rendererClass=&NiagaraMeshRendererProperties, rendererIndex=1}
  ]
 }
 BeamLightning {
  emitterUpdateScript [
   &EmitterState()
   &SpawnRate(
    SpawnRate=&FloatFromCurve(
     FloatCurve=DataInterface() {propertyValues {Curve {keys [
         {arriveTangent=6.637, interpMode=RCIM_Cubic, leaveTangent=6.637, tangentMode=RCTM_User, time=0, value=0}
         {arriveTangent=16.8545, leaveTangent=-0.0176, time=0.05, value=1}
         <at 0.6 1>
         {arriveTangent=-3, leaveTangent=-3, time=0.65, value=0.15}
         {arriveTangent=-0.4286, interpMode=RCIM_Cubic, leaveTangent=-0.4286, time=0.9, value=0.15}
         {arriveTangent=-2.5182, tangentMode=RCTM_User, time=1, value=0}
        ]}}}
     CurveIndex=$System.Timeline.NormalizedTime
     Scale Curve=20
    )
    Use Spawn Probability=false
   )
   &SpawnRate(
    SpawnGroup=1
    SpawnRate=&FloatFromCurve(
     FloatCurve=DataInterface() {propertyValues {Curve {keys [
         {arriveTangent=6.637, interpMode=RCIM_Cubic, leaveTangent=6.637, tangentMode=RCTM_User, time=0, value=0}
         {arriveTangent=16.8545, leaveTangent=-0.0176, time=0.05, value=1}
         <at 0.6 1>
         {arriveTangent=-3, leaveTangent=-3, time=0.65, value=0.15}
         {arriveTangent=-0.4286, interpMode=RCIM_Cubic, leaveTangent=-0.4286, time=0.9, value=0.15}
         {arriveTangent=-2.5182, tangentMode=RCTM_User, time=1, value=0}
        ]}}}
     CurveIndex=$System.Timeline.NormalizedTime
     Scale Curve=35
    )
    Use Spawn Probability=false
   ) {module=SpawnRate002}
  ]
  particleSpawnScript [
   &SkeletalMeshLocation(
    MaskSpawnGroup=false
    Mesh Sampling Type=::NewEnumerator5
    Position Sampling=::NewEnumerator2
    Skeletal Mesh=$System.Skeletal Mesh
    Socket Sampling Mode=::NewEnumerator5
   )
   &InitializeParticle(
    Color=&MergeHelper_ColorSwitch(
     0=&ColorFromCurve(
      ColorCurve=DataInterface() {propertyValues {
        AlphaCurve {keys [<at 0 1 RCIM_Cubic>,<at 1 1 RCIM_Cubic>]}
        BlueCurve {keys [<at 0.8 2 RCIM_Cubic>,<at 1 1 RCIM_Cubic>]}
        GreenCurve {keys [<at 0.8 0 RCIM_Cubic>,<at 1 0.8914 RCIM_Cubic>]}
        RedCurve {keys [<at 0.8 0.2667 RCIM_Cubic>,<at 1 0.0799 RCIM_Cubic>]}
       }}
      CurveIndex=$System.Timeline001.NormalizedTime
     )
     1=&ColorFromCurve(
      ColorCurve=DataInterface() {propertyValues {
        AlphaCurve {keys [<at 0 1 RCIM_Cubic>,<at 1 1 RCIM_Cubic>]}
        BlueCurve {keys [<at 0 0.1812 RCIM_Cubic>,<at 1 0.2894 RCIM_Cubic>]}
        GreenCurve {keys [<at 0 0.2462 RCIM_Cubic>,<at 1 0.4244 RCIM_Cubic>]}
        RedCurve {keys [<at 0 0.9473 RCIM_Cubic>,<at 1 0.9647 RCIM_Cubic>]}
       }}
      CurveIndex=$System.Timeline001.NormalizedTime
     )
     2=&ColorFromCurve(
      ColorCurve=DataInterface() {propertyValues {
        AlphaCurve {keys [<at 0 1 RCIM_Cubic>,<at 1 1 RCIM_Cubic>]}
        BlueCurve {keys [<at 0 0.0242>, {arriveTangent=0.1603, interpMode=RCIM_Cubic, leaveTangent=0.1603, time=0.25, value=0.0375}, <at 1 0.1845 RCIM_Cubic>]}
        GreenCurve {keys [<at 0 0.0176>, {arriveTangent=0.422, interpMode=RCIM_Cubic, leaveTangent=0.422, time=0.25, value=0.0849}, <at 1 0.4397 RCIM_Cubic>]}
        RedCurve {keys [<at 0 0.0232>, {arriveTangent=0.6564, interpMode=RCIM_Cubic, leaveTangent=0.6564, time=0.25, value=0.15}, <at 1 0.6795 RCIM_Cubic>]}
       }}
      CurveIndex=$System.Timeline001.NormalizedTime
     )
     3=&ColorFromCurve(
      ColorCurve=DataInterface() {propertyValues {
        AlphaCurve {keys [<at 0 1 RCIM_Cubic>,<at 1 1 RCIM_Cubic>]}
        BlueCurve {keys [<at 0 0.5647 RCIM_Cubic>,<at 1 0.7379 RCIM_Cubic>]}
        GreenCurve {keys [<at 0 0.0737 RCIM_Cubic>,<at 1 0.5843 RCIM_Cubic>]}
        RedCurve {keys [<at 0 0.5434 RCIM_Cubic>,<at 1 0.0834 RCIM_Cubic>]}
       }}
      CurveIndex=$System.Timeline001.NormalizedTime
     )
     4=&ColorFromCurve(
      ColorCurve=DataInterface() {propertyValues {
        AlphaCurve {keys [<at 0 1 RCIM_Cubic>,<at 1 1 RCIM_Cubic>]}
        BlueCurve {keys [<at 0.006 0.0802 RCIM_Cubic>,<at 0.25 0.4125>,<at 1 0.807 RCIM_Cubic>]}
        GreenCurve {keys [<at 0.006 0.0048 RCIM_Cubic>,<at 0.25 0.0999>,<at 1 0.4447 RCIM_Cubic>]}
        RedCurve {keys [<at 0.006 0.0015 RCIM_Cubic>,<at 0.25 0.0232>,<at 1 0.0807 RCIM_Cubic>]}
       }}
      CurveIndex=$System.Timeline001.NormalizedTime
     )
     5=&ColorFromCurve(
      ColorCurve=DataInterface() {propertyValues {
        AlphaCurve {keys [<at 0 1 RCIM_Cubic>,<at 1 1 RCIM_Cubic>]}
        BlueCurve {keys [<at 0 0.5271 RCIM_Cubic>,<at 1 0.6654 RCIM_Cubic>]}
        GreenCurve {keys [<at 0 0.0844 RCIM_Cubic>,<at 1 0.2961 RCIM_Cubic>]}
        RedCurve {keys [<at 0 0.3185 RCIM_Cubic>,<at 1 0.5583 RCIM_Cubic>]}
       }}
      CurveIndex=$System.Timeline001.NormalizedTime
     )
     Index=&MakeIntFromFloat(Float=$User.VariantIndex)
     Num=::NewEnumerator4
    )
    Lifetime=&MergeHelper_FloatSwitch(
     0=&RandomRangeFloat(Minimum=0.3, Maximum=0.5)
     1=&RandomRangeFloat(Minimum=0.15, Maximum=0.25)
     2=?
     3=?
     Index=&SpawnGroup()
     Num=::NewEnumerator0
    )
    Mesh Scale=&Multiply_VectorByFloat(
     Vector=&RandomRangeVector(Minimum=<0.5 0.5 1>, Maximum=<0.65 0.65 1>)
     Float=&MergeHelper_FloatSwitch(
      0=&RandomRangeFloat(Minimum=0.1, Maximum=0.2)
      1=&RandomRangeFloat(Minimum=0.05, Maximum=0.1)
      2=?
      3=?
      Index=&SpawnGroup()
      Num=::NewEnumerator0
     )
    )
    Mesh Scale Mode=::NewEnumerator4
    UsePositionOffset=false
    Write Material Random=true
   )
   &AddVelocity(Module.Use Falloff Distance=false, Override Seed=false, Velocity Mode=::NewEnumerator1, Velocity Speed=0.1)
   &SkeletalMeshLocation(
    Bone / Socket Index=&MakeCustomIntFromBool(
     Bool=$User.SkinWeightCheck
     True INT=&MakeIntFromFloat(Float=&UniformAOrBFloat(
       A=&FloatFromCurve(FloatCurve=DataInterface() {propertyValues {Curve {keys [<at 0 6>,<at 0.4 11>]}}}, CurveIndex=$System.Timeline001.NormalizedTime)
       B=&FloatFromCurve(FloatCurve=DataInterface() {propertyValues {Curve {keys [<at 0 6>,<at 0.4 10>]}}}, CurveIndex=$System.Timeline001.NormalizedTime)
       Evaluation Type=::NewEnumerator0
       Recalculate Random Each Loop=true
       Override Seed=false
      ))
     False INT=&MakeIntFromFloat(Float=&FloatFromCurve(
       FloatCurve=DataInterface() {propertyValues {Curve {keys [<at 0 0>,<at 0.4 1>]}}}
       CurveIndex=$System.Timeline001.NormalizedTime
       Scale Curve=&RandomRangeFloat(Minimum=5, Maximum=6)
      ))
    )
    MaskSpawnGroup=false
    Mesh Sampling Type=::NewEnumerator5
    Sampled Position Offset=&Lerp_Vector(
     A=&RandomRangeVector(Minimum=<0 -7.5 5>, Maximum=<0 7.5 5>)
     B=&RandomRangeVector(Minimum=<0 -12 0>, Maximum=<0 12 0>)
     Alpha=&FloatFromCurve(
      FloatCurve=DataInterface() {propertyValues {Curve {keys [
          {arriveTangent=0.2778, interpMode=RCIM_Cubic, leaveTangent=0.2778, tangentMode=RCTM_User, time=0.2, value=0}
          {arriveTangent=0.6216, interpMode=RCIM_Cubic, leaveTangent=0.2299, tangentMode=RCTM_Break, time=0.4, value=1.5}
          <at 0.45 1>
         ]}}}
      CurveIndex=$System.Timeline001.NormalizedTime
     )
    )
    Skeletal Mesh=$System.Skeletal Mesh
    Socket Sampling Mode=::NewEnumerator5
   ) {module=SkeletalMeshLocation002}
   /SetVariables(Particles.VisibilityTag=&SpawnGroup()) {isSetParameters=true}
  ]
  particleUpdateScript [
   &ParticleState()
   &ScaleMeshSize(Scale Factor=&MakeVector(
     X=1
     Y=1
     Z=&FloatFromCurve(
      FloatCurve=DataInterface() {propertyValues {Curve {keys [
          {interpMode=RCIM_Cubic, leaveTangent=0.0966, tangentMode=RCTM_Break, time=0, value=1}
          {arriveTangent=0.0951, interpMode=RCIM_Cubic, leaveTangent=-0.1745, tangentMode=RCTM_Break, time=1, value=1.07}
         ]}}}
      Scale Curve=&FN_FrontEndFloatModifier(Gameplay=1, Front End=$User.FrontEndScale)
     )
    ))
   &DynamicMaterialParameters(
    Index 0 Param 2=&RandomRangeFloat(Minimum=0, Maximum=2)
    Index 0 Param 4=&MakeCustomFloatFromBool(
     Bool=false
     True Float=1
     False Float=&SelectFloatFromArray(
      Array Sampling Mode=::NewEnumerator0
      Direct Array Index=?
      Float Selection Array=DataInterface() {propertyValues {FloatData [3,1,2,1]}}
      Random Seed=0
     )
    )
    Index 1 Param 1=1
    Index 1 Param 2=2
    Index 1 Param 3=&FloatFromCurve(FloatCurve=DataInterface() {propertyValues {Curve {keys [{interpMode=RCIM_Cubic, tangentMode=RCTM_Break, time=0.5, value=-0.5}, <at 1 0.1>]}}})
    Index 2 Param 1=&Multiply_Float(
     A=&FloatFromCurve(
      FloatCurve=DataInterface() {propertyValues {Curve {keys [{interpMode=RCIM_Cubic, leaveTangent=8.1085, tangentMode=RCTM_Break, time=0.8568, value=1}, {tangentMode=RCTM_User, time=1, value=0}]}}}
      Scale Curve=0.5
     )
     B=&Divide_Float(
      A=&MakeFloatFromVector(Vector=$Particles.Scale, Channel=::NewEnumerator0)
      B=&RandomRangeFloat(Minimum=10, Maximum=15)
     )
    )
    Index 2 Param 2=&FloatFromCurve(
     FloatCurve=DataInterface() {propertyValues {Curve {keys [
         {interpMode=RCIM_Cubic, tangentMode=RCTM_Break, time=0.7, value=0}
         {arriveTangent=-0.0859, interpMode=RCIM_Cubic, tangentMode=RCTM_Break, time=1, value=-1}
        ]}}}
     Scale Curve=0
    )
    Index 2 Param 3=&Waveform(
     Global Input Position=&RandomRangeFloat(Minimum=0, Maximum=15)
     Global Phase=$Particles.Age
     "Waveform [1]"=::NewEnumerator12
     "[1] Phase Scale"=2
     "[1] Frequency"=1
     "[1] Amplitude Min/Max"=<0 1>
    )
    Index 2 Param 4=25
    Index 3 Param 1=&FloatFromCurve(
     FloatCurve=DataInterface() {propertyValues {Curve {keys [
         {interpMode=RCIM_Cubic, leaveTangent=-0.0025, tangentMode=RCTM_Break, time=0, value=0}
         {arriveTangent=1.8528, interpMode=RCIM_Cubic, tangentMode=RCTM_Break, time=1, value=1}
        ]}}}
     Scale Curve=&RandomRangeFloat(Minimum=15, Maximum=20)
    )
    Index 3 Param 2=&FloatFromCurve(FloatCurve=DataInterface() {propertyValues {Curve {keys [
         {interpMode=RCIM_Cubic, leaveTangent=-0.0025, tangentMode=RCTM_Break, time=0, value=0.5}
         {arriveTangent=-0.0316, interpMode=RCIM_Cubic, tangentMode=RCTM_Break, time=1, value=-0.2}
        ]}}})
    Index 3 Param 3=&FloatFromCurve(FloatCurve=DataInterface() {propertyValues {Curve {keys [
         {interpMode=RCIM_Cubic, leaveTangent=1.1034, tangentMode=RCTM_Break, time=0, value=-1}
         {arriveTangent=2.2399, tangentMode=RCTM_User, time=0.5, value=0.125}
        ]}}})
    Index 3 Param 4=&FloatFromCurve(FloatCurve=DataInterface() {propertyValues {Curve {keys [
         {interpMode=RCIM_Cubic, leaveTangent=3.0601, tangentMode=RCTM_Break, time=0.5, value=-0.5}
         {arriveTangent=3.2053, tangentMode=RCTM_User, time=1, value=1}
        ]}}})
    Module.Index0Param0Write=true
    Module.Index0Param1Write=true
    Module.Index0Param2Write=true
    Module.Index0Param3Write=true
    Module.Index1Param0Write=true
    Module.Index1Param1Write=true
    Module.Index1Param2Write=true
    Module.Index1Param3Write=true
    Module.Index2Param0Write=true
    Module.Index2Param1Write=true
    Module.Index2Param2Write=true
    Module.Index2Param3Write=true
    Module.Index3Param0Write=true
    Module.Index3Param1Write=true
    Module.Index3Param2Write=true
    Module.Index3Param3Write=true
    Write Parameter Index 1=true
    Write Parameter Index 2=true
    Write Parameter Index 3=true
   )
   &SolveForcesAndVelocity(Clamp Velocity=false, Limit Acceleration=false)
   &ScaleColor(
    Scale Alpha=&FloatFromCurve(FloatCurve=DataInterface() {propertyValues {Curve {keys [
         {tangentMode=RCTM_User, time=0, value=1}
         {arriveTangent=-0.0405, interpMode=RCIM_Cubic, leaveTangent=-0.0405, tangentMode=RCTM_User, time=0.7009, value=1}
         {arriveTangent=-8.0089, interpMode=RCIM_Cubic, leaveTangent=-8.0089, tangentMode=RCTM_User, time=1, value=0}
        ]}}})
    Scale RGB=&Lerp_Vector(
     A=&VectorFromFloat(Value=&FloatFromCurve(FloatCurve=DataInterface() {propertyValues {Curve {keys [<at 0.2771 1>,<at 0.3786 1>]}}}, Scale Curve=20))
     B=&VectorFromFloat(Value=&RandomRangeFloat(Minimum=0.5, Maximum=1.5, Evaluation Type=::NewEnumerator1))
     Alpha=&FloatFromCurve(FloatCurve=DataInterface() {propertyValues {Curve {keys [
          {interpMode=RCIM_Cubic, leaveTangent=2.5294, tangentMode=RCTM_Break, time=0, value=0}
          {arriveTangent=-0.2166, interpMode=RCIM_Cubic, tangentMode=RCTM_Break, time=0.4, value=1}
         ]}}})
    )
    ScaleA=true
    ScaleRGB=true
   )
   /SetVariables(Particles.MeshOrientation=&MultiplyQuaternion(
     Quaternion A=&OrientMeshToVector(Look At Direction=$Particles.Velocity, Reference Axis Vector=<1 0 0>)
     Quaternion B=$Particles.SkeletalMeshLocation.SampledRotation
    )) {isSetParameters=true}
   &CameraOffset(Camera Offset Amount=-2)
   &FN_Sprite_Attributes(
    Camera Offset=15
    Color=&MakeLinearColorFromVectorAndFloat(
     "Vector (RGB)"=&MakeVectorFromLinearColorRGB(Color=&Multiply_LinearColorByFloat(
       Linear Color=&MergeHelper_ColorSwitch(
        0=&ColorFromCurve(
         ColorCurve=DataInterface() {propertyValues {
           AlphaCurve {keys [<at 0 1 RCIM_Cubic>,<at 1 1 RCIM_Cubic>]}
           BlueCurve {keys [<at 0.8 2 RCIM_Cubic>,<at 1 1 RCIM_Cubic>]}
           GreenCurve {keys [<at 0.8 0 RCIM_Cubic>,<at 1 0.8914 RCIM_Cubic>]}
           RedCurve {keys [<at 0.8 0.2667 RCIM_Cubic>,<at 1 0.0799 RCIM_Cubic>]}
          }}
         CurveIndex=$System.Timeline001.NormalizedTime
        )
        1=&ColorFromCurve(
         ColorCurve=DataInterface() {propertyValues {
           AlphaCurve {keys [<at 0 1 RCIM_Cubic>,<at 1 1 RCIM_Cubic>]}
           BlueCurve {keys [<at 0 0.1812 RCIM_Cubic>,<at 1 0.2894 RCIM_Cubic>]}
           GreenCurve {keys [<at 0 0.2462 RCIM_Cubic>,<at 1 0.4244 RCIM_Cubic>]}
           RedCurve {keys [<at 0 0.9473 RCIM_Cubic>,<at 1 0.9647 RCIM_Cubic>]}
          }}
         CurveIndex=$System.Timeline001.NormalizedTime
        )
        2=&ColorFromCurve(
         ColorCurve=DataInterface() {propertyValues {
           AlphaCurve {keys [<at 0 1 RCIM_Cubic>,<at 1 1 RCIM_Cubic>]}
           BlueCurve {keys [<at 0 0.0242>, {arriveTangent=0.1603, interpMode=RCIM_Cubic, leaveTangent=0.1603, time=0.25, value=0.0375}, <at 1 0.1845 RCIM_Cubic>]}
           GreenCurve {keys [<at 0 0.0176>, {arriveTangent=0.422, interpMode=RCIM_Cubic, leaveTangent=0.422, time=0.25, value=0.0849}, <at 1 0.4397 RCIM_Cubic>]}
           RedCurve {keys [<at 0 0.0232>, {arriveTangent=0.6564, interpMode=RCIM_Cubic, leaveTangent=0.6564, time=0.25, value=0.15}, <at 1 0.6795 RCIM_Cubic>]}
          }}
         CurveIndex=$System.Timeline001.NormalizedTime
        )
        3=&ColorFromCurve(
         ColorCurve=DataInterface() {propertyValues {
           AlphaCurve {keys [<at 0 1 RCIM_Cubic>,<at 1 1 RCIM_Cubic>]}
           BlueCurve {keys [<at 0 0.5647 RCIM_Cubic>,<at 1 0.7379 RCIM_Cubic>]}
           GreenCurve {keys [<at 0 0.0737 RCIM_Cubic>,<at 1 0.5843 RCIM_Cubic>]}
           RedCurve {keys [<at 0 0.5434 RCIM_Cubic>,<at 1 0.0834 RCIM_Cubic>]}
          }}
         CurveIndex=$System.Timeline001.NormalizedTime
        )
        4=&ColorFromCurve(
         ColorCurve=DataInterface() {propertyValues {
           AlphaCurve {keys [<at 0 1 RCIM_Cubic>,<at 1 1 RCIM_Cubic>]}
           BlueCurve {keys [<at 0.006 0.0802 RCIM_Cubic>,<at 0.25 0.4125>,<at 1 0.807 RCIM_Cubic>]}
           GreenCurve {keys [<at 0.006 0.0048 RCIM_Cubic>,<at 0.25 0.0999>,<at 1 0.4447 RCIM_Cubic>]}
           RedCurve {keys [<at 0.006 0.0015 RCIM_Cubic>,<at 0.25 0.0232>,<at 1 0.0807 RCIM_Cubic>]}
          }}
         CurveIndex=$System.Timeline001.NormalizedTime
        )
        5=&ColorFromCurve(
         ColorCurve=DataInterface() {propertyValues {
           AlphaCurve {keys [<at 0 1 RCIM_Cubic>,<at 1 1 RCIM_Cubic>]}
           BlueCurve {keys [<at 0 0.5271 RCIM_Cubic>,<at 1 0.6654 RCIM_Cubic>]}
           GreenCurve {keys [<at 0 0.0844 RCIM_Cubic>,<at 1 0.2961 RCIM_Cubic>]}
           RedCurve {keys [<at 0 0.3185 RCIM_Cubic>,<at 1 0.5583 RCIM_Cubic>]}
          }}
         CurveIndex=$System.Timeline001.NormalizedTime
        )
        Index=&MakeIntFromFloat(Float=$User.VariantIndex)
        Num=::NewEnumerator4
       )
       Float=0.5
      ))
     "Float (Alpha)"=&Multiply_Float(
      A=&FloatFromCurve(
       FloatCurve=DataInterface() {propertyValues {Curve {keys [
           {interpMode=RCIM_Cubic, tangentMode=RCTM_User, time=0, value=0}
           {interpMode=RCIM_Cubic, tangentMode=RCTM_User, time=0.3, value=1}
           {tangentMode=RCTM_User, time=1, value=0}
          ]}}}
       Scale Curve=0.7
      )
      B=&MakeCustomFloatFromBool(Bool=$User.bIsHeadHidden, True Float=0, False Float=1)
     )
    )
    Create Camera Offset Attribute=true
    Create Color Attribute=true
    Create Dynamic Material Parameter Attributes=true
    Create SpriteSize Attribute=true
    Index 0 Param 1=1
    Index 0 Param 2=5
    Index 0 Param 4:NiagaraFloat=&RandomRangeFloat(Minimum=0, Maximum=1)
    Index 3 Param 1=1
    Module Output:ENiagara_AttributeSamplingApplyOutput=::NewEnumerator1
    Module.Index0Float0Write=true
    Module.Index0Float1Write=true
    Module.Index0Float2Write=true
    Module.Index0Float3Write=true
    SpriteSize=&Multiply_Vector2D_ByFloat(
     Float=&FloatFromCurve(FloatCurve=DataInterface() {propertyValues {Curve {keys [
          {interpMode=RCIM_Cubic, tangentMode=RCTM_User, time=0, value=0}
          {interpMode=RCIM_Cubic, tangentMode=RCTM_User, time=0.3, value=1}
          {tangentMode=RCTM_User, time=1, value=0}
         ]}}})
     Vector2D=<15 15>
    )
    Write Parameter Index 3:ENiagara_DynamicMaterialParametersInputMode=::NewEnumerator1
   )
   &ScaleColor(
    Scale Alpha=&MakeCustomFloatFromBool(Bool=$User.bIsHeadHidden, True Float=0, False Float=1)
    ScaleA=true
    ScaleRGB=false
   ) {module=ScaleColor001}
  ]
  rendererClasses [&NiagaraMeshRendererProperties, &NiagaraSpriteRendererProperties]
  renderers [
   {rendererClass=&NiagaraMeshRendererProperties, rendererIndex=0}
   {rendererClass=&NiagaraMeshRendererProperties, rendererIndex=1}
   {rendererClass=&NiagaraSpriteRendererProperties, rendererIndex=2}
  ]
 }
 Glow001 {
  emitterUpdateScript [
   &EmitterState(Life Cycle Mode=::NewEnumerator1, Loop Behavior=::NewEnumerator1, Loop Duration=1, UseLoopDelay=false)
   &SpawnBurst_Instantaneous(Spawn Probability=$User.IsReactive, Use Spawn Probability=true, UseLoopCountLimit=false)
  ]
  particleSpawnScript [&InitializeParticle(
    Color=&SelectLinearColorFromArray(
     Array Sampling Mode=::NewEnumerator1
     Direct Array Index=&MakeIntFromFloat(Float=$User.VariantIndex)
     Color Selection Array=DataInterface() {propertyValues {ColorData [
        <rgba 0.3913 0.9258 1 1>
        <rgba 0.9647 0.502 0.3859 1>
        <rgba 0.6795 0.482 0.2718 1>
        <rgba 0.5974 0.5335 0.7379 1>
        <rgba 0.3228 0.5659 0.807 1>
        <rgba 0.5583 0.2961 0.6654 1>
       ]}}
     Random Seed=?
    )
    Lifetime=5
    Sprite Size=<55 42>
    Sprite Size Mode=::NewEnumerator4
    UsePositionOffset=false
    Write Material Random=true
   )]
  particleUpdateScript [
   &ParticleState()
   &SkeletalMeshLocation(
    Bone / Socket Index=&MakeCustomIntFromBool(Bool=$User.SkinWeightCheck, True INT=11, False INT=5)
    MaskSpawnGroup=false
    Mesh Sampling Type=::NewEnumerator5
    Sampled Position Offset=&MakeCustomVectorFromBool(Bool=$User.SkinWeightCheck, True Vector=<0 0 0>, False Vector=<0 0 0>)
    Skeletal Mesh=$System.Skeletal Mesh
    Socket Sampling Mode=::NewEnumerator5
   ) {module=SkeletalMeshLocation001}
   &ScaleColor(
    Scale Alpha=&FloatFromCurve(
     FloatCurve=DataInterface() {propertyValues {Curve {
        defaultValue=1
        keys [
         {interpMode=RCIM_Cubic, tangentMode=RCTM_User, time=0, value=0}
         <at 0.1 1>
         {arriveTangent=-1.1111, interpMode=RCIM_Cubic, leaveTangent=-1.1111, time=0.9, value=1}
         {tangentMode=RCTM_User, time=1, value=0}
        ]
       }}}
     CurveIndex=$System.Timeline.NormalizedTime
     Scale Curve=&Multiply_Float(A=&FNTimeOfDayScalar(Night Scale=0.65, Morning Scale=0.8, Evening Scale=0.8), B=1)
    )
    ScaleA=true
    ScaleRGB=false
   )
   &ScaleColor(
    Scale Alpha=&MakeCustomFloatFromBool(Bool=$User.bIsHeadHidden, True Float=0, False Float=0.75)
    ScaleA=true
    ScaleRGB=false
   ) {module=ScaleColor002}
  ]
  rendererClasses [&NiagaraSpriteRendererProperties]
  renderers [{rendererClass=&NiagaraSpriteRendererProperties}]
 }
}
summary {
 systemName=NS_Character_ThunderBallet_Idle
 emitters [
  {emitterName=BeamLightning, rendererClasses [&NiagaraMeshRendererProperties, &NiagaraSpriteRendererProperties]}
  {emitterName=ArcLightning, rendererClasses [&NiagaraMeshRendererProperties]}
  {emitterName=Glow001, rendererClasses [&NiagaraSpriteRendererProperties]}
 ]
 userVariables [
  User.bIsFrontEnd {type=NiagaraBool}
  User.bIsFrontEndPreview {type=NiagaraBool}
  User.bIsHeadHidden {type=NiagaraBool}
  User.FrontEndScale {type=NiagaraFloat, defaultValue=0.6, description=""}
  User.IsReactive {type=NiagaraFloat, defaultValue=0, description=""}
  User.KillCount {type=NiagaraInt32}
  User.SkinWeightCheck {type=NiagaraBool}
  User.VariantIndex {type=NiagaraFloat, defaultValue=0, description=""}
 ]
}
