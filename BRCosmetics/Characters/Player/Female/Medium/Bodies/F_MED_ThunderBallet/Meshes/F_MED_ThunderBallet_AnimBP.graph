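# Post-copy Anim Blueprint for the female medium ThunderBallet body cosmetic, adding skirt motion to the attached parent's pose.
# The copied pose feeds a Control Rig driven by crouch, knee-slide and baseball-slide weights; crouch snaps, slides interpolate.
# Chained AnimDynamics nodes then simulate left, right and center skirt bones with angular springs and pelvis/thigh sphere limits.
# Center bones reuse the left-side limits, so they ignore the right thigh.
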AnimBlueprint /BRCosmetics/Characters/Player/Female/Medium/Bodies/F_MED_ThunderBallet/Meshes/F_MED_ThunderBallet_AnimBP
target_skeleton=&/F_MED_ThunderBallet_Skeleton
graphs [
 AnimGraph {nodes [
   n1 = Control Rig!(
    Source=Copy Pose From Mesh() {bUseAttachedParent=true}
    Alpha
    isCrouching:double=Get is crouching() -> (is crouching:double)
    is knee sliding:double=Get is knee sliding() -> (is knee sliding:double)
    is baseball sliding:double=Get is baseball sliding() -> (is baseball sliding:double)
   ) -> (Pose) {controlRigAssetReference {blueprintRigClass=&/F_MED_ThunderBallet_ControlRig.F_MED_ThunderBallet_ControlRig_C}}
   n2 = Anim Dynamics - Bone: dyn_skirt_fr_01_R!(ComponentPose=Local To Component(LocalPose=@n1), Alpha=1, GravityOverride:Vector="0,0,-980") -> (Pose) {
    angularDampingOverride=0.999
    angularSpringConstant=500
    bAngularSpring=true
    bOverrideAngularDamping=true
    bUseGravityOverride=true
    simulationSpace=RootRelative
    boundBone {boneName=dyn_skirt_fr_01_R}
    componentAppliedLinearAccClamp=<3000 3000 3000>
    componentLinearAccScale=<0.5 0.5 0.5>
    componentLinearVelScale=<0.5 0.5 0.5>
    physicsBodyDefinitions [{
      collisionType=CustomSphere
      sphereCollisionRadius=2
      boundBone {boneName=dyn_skirt_fr_01_R}
      boxExtents=<25 25 25>
      constraintSetup {angularLimitsMax=<0 0 0>, angularLimitsMin=<0 -15 -5>, angularTarget=<1 0 0>}
      localJointOffset=<3 0 0>
     }]
    sphericalLimits [
     {limitRadius=8, drivingBone {boneName=thigh_twist_01_r}, sphereLocalOffset=<-15 0 0>}
     {limitRadius=10.5, drivingBone {boneName=pelvis}, sphereLocalOffset=<0 0 0>}
     {limitRadius=8, drivingBone {boneName=thigh_twist_01_r}, sphereLocalOffset=<-25 0 0>}
     {limitRadius=8, drivingBone {boneName=thigh_twist_01_r}, sphereLocalOffset=<-5 0 0>}
    ]
   }
   n3 = Anim Dynamics - Bone: dyn_skirt_fr_02_R!(ComponentPose=@n2, Alpha=1, GravityOverride:Vector="0,0,-980") -> (Pose) {
    angularDampingOverride=0.999
    angularSpringConstant=500
    bAngularSpring=true
    bOverrideAngularDamping=true
    bUseGravityOverride=true
    simulationSpace=RootRelative
    boundBone {boneName=dyn_skirt_fr_02_R}
    componentAppliedLinearAccClamp=<3000 3000 3000>
    componentLinearAccScale=<0.5 0.5 0.5>
    componentLinearVelScale=<0.5 0.5 0.5>
    physicsBodyDefinitions [{
      collisionType=CustomSphere
      sphereCollisionRadius=2
      boundBone {boneName=dyn_skirt_fr_02_R}
      boxExtents=<15 15 15>
      constraintSetup {angularLimitsMax=<0 0 0>, angularLimitsMin=<0 -15 -5>, angularTarget=<1 0 0>}
      localJointOffset=<3 0 0>
     }]
    sphericalLimits [
     {limitRadius=8, drivingBone {boneName=thigh_twist_01_r}, sphereLocalOffset=<-15 0 0>}
     {limitRadius=10.5, drivingBone {boneName=pelvis}, sphereLocalOffset=<0 0 0>}
     {limitRadius=8, drivingBone {boneName=thigh_twist_01_r}, sphereLocalOffset=<-25 0 0>}
     {limitRadius=8, drivingBone {boneName=thigh_twist_01_r}, sphereLocalOffset=<-5 0 0>}
    ]
   }
   n4 = Anim Dynamics - Bone: dyn_skirt_01_R!(ComponentPose=@n3, Alpha=1, GravityOverride:Vector="0,0,-980") -> (Pose) {
    angularDampingOverride=0.999
    angularSpringConstant=500
    bAngularSpring=true
    bOverrideAngularDamping=true
    bUseGravityOverride=true
    simulationSpace=RootRelative
    boundBone {boneName=dyn_skirt_01_R}
    componentAppliedLinearAccClamp=<3000 3000 3000>
    componentLinearAccScale=<0.5 0.5 0.5>
    componentLinearVelScale=<0.5 0.5 0.5>
    physicsBodyDefinitions [{
      collisionType=CustomSphere
      sphereCollisionRadius=2
      boundBone {boneName=dyn_skirt_01_R}
      boxExtents=<25 25 25>
      constraintSetup {angularLimitsMax=<0 0 5>, angularLimitsMin=<0 -15 -5>, angularTarget=<1 0 0>}
      localJointOffset=<3 0 0>
     }]
    sphericalLimits [
     {limitRadius=8, drivingBone {boneName=thigh_twist_01_r}, sphereLocalOffset=<-15 0 0>}
     {limitRadius=10.5, drivingBone {boneName=pelvis}, sphereLocalOffset=<0 0 0>}
     {limitRadius=8, drivingBone {boneName=thigh_twist_01_r}, sphereLocalOffset=<-25 0 0>}
     {limitRadius=8, drivingBone {boneName=thigh_twist_01_r}, sphereLocalOffset=<-5 0 0>}
    ]
   }
   n5 = Anim Dynamics - Bone: dyn_skirt_02_R!(ComponentPose=@n4, Alpha=1, GravityOverride:Vector="0,0,-980") -> (Pose) {
    angularDampingOverride=0.999
    angularSpringConstant=500
    bAngularSpring=true
    bOverrideAngularDamping=true
    bUseGravityOverride=true
    simulationSpace=RootRelative
    boundBone {boneName=dyn_skirt_02_R}
    componentAppliedLinearAccClamp=<3000 3000 3000>
    componentLinearAccScale=<0.5 0.5 0.5>
    componentLinearVelScale=<0.5 0.5 0.5>
    physicsBodyDefinitions [{
      collisionType=CustomSphere
      sphereCollisionRadius=2
      boundBone {boneName=dyn_skirt_02_R}
      boxExtents=<15 15 15>
      constraintSetup {angularLimitsMax=<0 0 5>, angularLimitsMin=<0 -15 -5>, angularTarget=<1 0 0>}
      localJointOffset=<3 0 0>
     }]
    sphericalLimits [
     {limitRadius=8, drivingBone {boneName=thigh_twist_01_r}, sphereLocalOffset=<-15 0 0>}
     {limitRadius=10.5, drivingBone {boneName=pelvis}, sphereLocalOffset=<0 0 0>}
     {limitRadius=8, drivingBone {boneName=thigh_twist_01_r}, sphereLocalOffset=<-25 0 0>}
     {limitRadius=8, drivingBone {boneName=thigh_twist_01_r}, sphereLocalOffset=<-5 0 0>}
    ]
   }
   n6 = Anim Dynamics - Bone: dyn_skirt_bk_01_R!(ComponentPose=@n5, Alpha=1, GravityOverride:Vector="0,0,-980") -> (Pose) {
    angularDampingOverride=0.999
    angularSpringConstant=500
    bAngularSpring=true
    bOverrideAngularDamping=true
    bUseGravityOverride=true
    simulationSpace=RootRelative
    boundBone {boneName=dyn_skirt_bk_01_R}
    componentAppliedLinearAccClamp=<3000 3000 3000>
    componentLinearAccScale=<0.5 0.5 0.5>
    componentLinearVelScale=<0.5 0.5 0.5>
    physicsBodyDefinitions [{
      collisionType=CustomSphere
      sphereCollisionRadius=2
      boundBone {boneName=dyn_skirt_bk_01_R}
      boxExtents=<25 25 25>
      constraintSetup {angularLimitsMax=<0 0 5>, angularLimitsMin=<0 -15 -5>, angularTarget=<1 0 0>}
      localJointOffset=<3 0 0>
     }]
    sphericalLimits [
     {limitRadius=8, drivingBone {boneName=thigh_twist_01_r}, sphereLocalOffset=<-15 0 0>}
     {limitRadius=10.5, drivingBone {boneName=pelvis}, sphereLocalOffset=<0 0 0>}
     {limitRadius=8, drivingBone {boneName=thigh_twist_01_r}, sphereLocalOffset=<-25 0 0>}
     {limitRadius=8, drivingBone {boneName=thigh_twist_01_r}, sphereLocalOffset=<-5 0 0>}
    ]
   }
   n7 = Anim Dynamics - Bone: dyn_skirt_bk_02_R!(ComponentPose=@n6, Alpha=1, GravityOverride:Vector="0,0,-980") -> (Pose) {
    angularDampingOverride=0.999
    angularSpringConstant=500
    bAngularSpring=true
    bOverrideAngularDamping=true
    bUseGravityOverride=true
    simulationSpace=RootRelative
    boundBone {boneName=dyn_skirt_bk_02_R}
    componentAppliedLinearAccClamp=<3000 3000 3000>
    componentLinearAccScale=<0.5 0.5 0.5>
    componentLinearVelScale=<0.5 0.5 0.5>
    physicsBodyDefinitions [{
      collisionType=CustomSphere
      sphereCollisionRadius=2
      boundBone {boneName=dyn_skirt_bk_02_R}
      boxExtents=<15 15 15>
      constraintSetup {angularLimitsMax=<0 0 5>, angularLimitsMin=<0 -15 -5>, angularTarget=<1 0 0>}
      localJointOffset=<3 0 0>
     }]
    sphericalLimits [
     {limitRadius=8, drivingBone {boneName=thigh_twist_01_r}, sphereLocalOffset=<-15 0 0>}
     {limitRadius=10.5, drivingBone {boneName=pelvis}, sphereLocalOffset=<0 0 0>}
     {limitRadius=8, drivingBone {boneName=thigh_twist_01_r}, sphereLocalOffset=<-25 0 0>}
     {limitRadius=8, drivingBone {boneName=thigh_twist_01_r}, sphereLocalOffset=<-5 0 0>}
    ]
   }
   n8 = Anim Dynamics - Bone: dyn_skirt_fr_01_L!(ComponentPose=@n7, Alpha=1, GravityOverride:Vector="0,0,-980") -> (Pose) {
    angularDampingOverride=0.999
    angularSpringConstant=500
    bAngularSpring=true
    bOverrideAngularDamping=true
    bUseGravityOverride=true
    simulationSpace=RootRelative
    boundBone {boneName=dyn_skirt_fr_01_L}
    componentAppliedLinearAccClamp=<3000 3000 3000>
    componentLinearAccScale=<0.5 0.5 0.5>
    componentLinearVelScale=<0.5 0.5 0.5>
    physicsBodyDefinitions [{
      collisionType=CustomSphere
      sphereCollisionRadius=2
      boundBone {boneName=dyn_skirt_fr_01_L}
      boxExtents=<25 25 25>
      constraintSetup {angularLimitsMax=<0 0 0>, angularLimitsMin=<0 -15 -5>, angularTarget=<1 0 0>}
      localJointOffset=<3 0 0>
     }]
    sphericalLimits [
     {limitRadius=7.5, drivingBone {boneName=thigh_twist_01_l}, sphereLocalOffset=<15 0 0>}
     {limitRadius=11, drivingBone {boneName=pelvis}, sphereLocalOffset=<-1 0 0>}
     {limitRadius=8, drivingBone {boneName=thigh_twist_01_l}, sphereLocalOffset=<25 0 0>}
     {limitRadius=8, drivingBone {boneName=thigh_twist_01_l}, sphereLocalOffset=<5 0 0>}
    ]
   }
   n9 = Anim Dynamics - Bone: dyn_skirt_fr_02_L!(ComponentPose=@n8, Alpha=1, GravityOverride:Vector="0,0,-980") -> (Pose) {
    angularDampingOverride=0.999
    angularSpringConstant=500
    bAngularSpring=true
    bOverrideAngularDamping=true
    bUseGravityOverride=true
    simulationSpace=RootRelative
    boundBone {boneName=dyn_skirt_fr_02_L}
    componentAppliedLinearAccClamp=<3000 3000 3000>
    componentLinearAccScale=<0.5 0.5 0.5>
    componentLinearVelScale=<0.5 0.5 0.5>
    physicsBodyDefinitions [{
      collisionType=CustomSphere
      sphereCollisionRadius=2
      boundBone {boneName=dyn_skirt_fr_02_L}
      boxExtents=<15 15 15>
      constraintSetup {angularLimitsMax=<0 0 0>, angularLimitsMin=<0 -15 -5>, angularTarget=<1 0 0>}
      localJointOffset=<3 0 0>
     }]
    sphericalLimits [
     {limitRadius=7.5, drivingBone {boneName=thigh_twist_01_l}, sphereLocalOffset=<15 0 0>}
     {limitRadius=11, drivingBone {boneName=pelvis}, sphereLocalOffset=<-1 0 0>}
     {limitRadius=8, drivingBone {boneName=thigh_twist_01_l}, sphereLocalOffset=<25 0 0>}
     {limitRadius=8, drivingBone {boneName=thigh_twist_01_l}, sphereLocalOffset=<5 0 0>}
    ]
   }
   n10 = Anim Dynamics - Bone: dyn_skirt_01_L!(ComponentPose=@n9, Alpha=1, GravityOverride:Vector="0,0,-980") -> (Pose) {
    angularDampingOverride=0.999
    angularSpringConstant=500
    bAngularSpring=true
    bOverrideAngularDamping=true
    bUseGravityOverride=true
    simulationSpace=RootRelative
    boundBone {boneName=dyn_skirt_01_L}
    componentAppliedLinearAccClamp=<3000 3000 3000>
    componentLinearAccScale=<0.5 0.5 0.5>
    componentLinearVelScale=<0.5 0.5 0.5>
    physicsBodyDefinitions [{
      collisionType=CustomSphere
      sphereCollisionRadius=2
      boundBone {boneName=dyn_skirt_01_L}
      boxExtents=<25 25 25>
      constraintSetup {angularLimitsMax=<0 0 5>, angularLimitsMin=<0 -15 -5>, angularTarget=<1 0 0>}
      localJointOffset=<3 0 0>
     }]
    sphericalLimits [
     {limitRadius=7.5, drivingBone {boneName=thigh_twist_01_l}, sphereLocalOffset=<15 0 0>}
     {limitRadius=11, drivingBone {boneName=pelvis}, sphereLocalOffset=<-1 0 0>}
     {limitRadius=8, drivingBone {boneName=thigh_twist_01_l}, sphereLocalOffset=<25 0 0>}
     {limitRadius=8, drivingBone {boneName=thigh_twist_01_l}, sphereLocalOffset=<5 0 0>}
    ]
   }
   n11 = Anim Dynamics - Bone: dyn_skirt_02_L!(ComponentPose=@n10, Alpha=1, GravityOverride:Vector="0,0,-980") -> (Pose) {
    angularDampingOverride=0.999
    angularSpringConstant=500
    bAngularSpring=true
    bOverrideAngularDamping=true
    bUseGravityOverride=true
    simulationSpace=RootRelative
    boundBone {boneName=dyn_skirt_02_L}
    componentAppliedLinearAccClamp=<3000 3000 3000>
    componentLinearAccScale=<0.5 0.5 0.5>
    componentLinearVelScale=<0.5 0.5 0.5>
    physicsBodyDefinitions [{
      collisionType=CustomSphere
      sphereCollisionRadius=2
      boundBone {boneName=dyn_skirt_02_L}
      boxExtents=<15 15 15>
      constraintSetup {angularLimitsMax=<0 0 5>, angularLimitsMin=<0 -15 -5>, angularTarget=<1 0 0>}
      localJointOffset=<3 0 0>
     }]
    sphericalLimits [
     {limitRadius=7.5, drivingBone {boneName=thigh_twist_01_l}, sphereLocalOffset=<15 0 0>}
     {limitRadius=11, drivingBone {boneName=pelvis}, sphereLocalOffset=<-1 0 0>}
     {limitRadius=8, drivingBone {boneName=thigh_twist_01_l}, sphereLocalOffset=<25 0 0>}
     {limitRadius=8, drivingBone {boneName=thigh_twist_01_l}, sphereLocalOffset=<5 0 0>}
    ]
   }
   n12 = Anim Dynamics - Bone: dyn_skirt_bk_01_L!(ComponentPose=@n11, Alpha=1, GravityOverride:Vector="0,0,-980") -> (Pose) {
    angularDampingOverride=0.999
    angularSpringConstant=500
    bAngularSpring=true
    bOverrideAngularDamping=true
    bUseGravityOverride=true
    simulationSpace=RootRelative
    boundBone {boneName=dyn_skirt_bk_01_L}
    componentAppliedLinearAccClamp=<3000 3000 3000>
    componentLinearAccScale=<0.5 0.5 0.5>
    componentLinearVelScale=<0.5 0.5 0.5>
    physicsBodyDefinitions [{
      collisionType=CustomSphere
      sphereCollisionRadius=2
      boundBone {boneName=dyn_skirt_bk_01_L}
      boxExtents=<25 25 25>
      constraintSetup {angularLimitsMax=<0 0 5>, angularLimitsMin=<0 -15 -5>, angularTarget=<1 0 0>}
      localJointOffset=<3 0 0>
     }]
    sphericalLimits [
     {limitRadius=7.5, drivingBone {boneName=thigh_twist_01_l}, sphereLocalOffset=<15 0 0>}
     {limitRadius=11, drivingBone {boneName=pelvis}, sphereLocalOffset=<-1 0 0>}
     {limitRadius=8, drivingBone {boneName=thigh_twist_01_l}, sphereLocalOffset=<25 0 0>}
     {limitRadius=8, drivingBone {boneName=thigh_twist_01_l}, sphereLocalOffset=<5 0 0>}
    ]
   }
   n13 = Anim Dynamics - Bone: dyn_skirt_bk_02_L!(ComponentPose=@n12, Alpha=1, GravityOverride:Vector="0,0,-980") -> (Pose) {
    angularDampingOverride=0.999
    angularSpringConstant=500
    bAngularSpring=true
    bOverrideAngularDamping=true
    bUseGravityOverride=true
    simulationSpace=RootRelative
    boundBone {boneName=dyn_skirt_bk_02_L}
    componentAppliedLinearAccClamp=<3000 3000 3000>
    componentLinearAccScale=<0.5 0.5 0.5>
    componentLinearVelScale=<0.5 0.5 0.5>
    physicsBodyDefinitions [{
      collisionType=CustomSphere
      sphereCollisionRadius=2
      boundBone {boneName=dyn_skirt_bk_02_L}
      boxExtents=<15 15 15>
      constraintSetup {angularLimitsMax=<0 0 5>, angularLimitsMin=<0 -15 -5>, angularTarget=<1 0 0>}
      localJointOffset=<3 0 0>
     }]
    sphericalLimits [
     {limitRadius=7.5, drivingBone {boneName=thigh_twist_01_l}, sphereLocalOffset=<15 0 0>}
     {limitRadius=11, drivingBone {boneName=pelvis}, sphereLocalOffset=<-1 0 0>}
     {limitRadius=8, drivingBone {boneName=thigh_twist_01_l}, sphereLocalOffset=<25 0 0>}
     {limitRadius=8, drivingBone {boneName=thigh_twist_01_l}, sphereLocalOffset=<5 0 0>}
    ]
   }
   n14 = Anim Dynamics - Bone: dyn_skirt_fr_01!(ComponentPose=@n13, Alpha=1, GravityOverride:Vector="0,0,-980") -> (Pose) {
    angularDampingOverride=0.999
    angularSpringConstant=500
    bAngularSpring=true
    bOverrideAngularDamping=true
    bUseGravityOverride=true
    simulationSpace=RootRelative
    boundBone {boneName=dyn_skirt_fr_01}
    componentAppliedLinearAccClamp=<3000 3000 3000>
    componentLinearAccScale=<0.5 0.5 0.5>
    componentLinearVelScale=<0.5 0.5 0.5>
    physicsBodyDefinitions [{
      collisionType=CustomSphere
      sphereCollisionRadius=2
      boundBone {boneName=dyn_skirt_fr_01}
      boxExtents=<25 25 25>
      constraintSetup {angularLimitsMax=<0 0 0>, angularLimitsMin=<0 -15 -5>, angularTarget=<1 0 0>}
      localJointOffset=<3 0 0>
     }]
    sphericalLimits [
     {limitRadius=7.5, drivingBone {boneName=thigh_twist_01_l}, sphereLocalOffset=<15 0 0>}
     {limitRadius=11, drivingBone {boneName=pelvis}, sphereLocalOffset=<-1 0 0>}
     {limitRadius=8, drivingBone {boneName=thigh_twist_01_l}, sphereLocalOffset=<25 0 0>}
     {limitRadius=8, drivingBone {boneName=thigh_twist_01_l}, sphereLocalOffset=<5 0 0>}
    ]
   }
   n15 = Anim Dynamics - Bone: dyn_skirt_fr_02!(ComponentPose=@n14, Alpha=1, GravityOverride:Vector="0,0,-980") -> (Pose) {
    angularDampingOverride=0.999
    angularSpringConstant=500
    bAngularSpring=true
    bOverrideAngularDamping=true
    bUseGravityOverride=true
    simulationSpace=RootRelative
    boundBone {boneName=dyn_skirt_fr_02}
    componentAppliedLinearAccClamp=<3000 3000 3000>
    componentLinearAccScale=<0.5 0.5 0.5>
    componentLinearVelScale=<0.5 0.5 0.5>
    physicsBodyDefinitions [{
      collisionType=CustomSphere
      sphereCollisionRadius=2
      boundBone {boneName=dyn_skirt_fr_02}
      boxExtents=<15 15 15>
      constraintSetup {angularLimitsMax=<0 0 0>, angularLimitsMin=<0 -15 -5>, angularTarget=<1 0 0>}
      localJointOffset=<3 0 0>
     }]
    sphericalLimits [
     {limitRadius=7.5, drivingBone {boneName=thigh_twist_01_l}, sphereLocalOffset=<15 0 0>}
     {limitRadius=11, drivingBone {boneName=pelvis}, sphereLocalOffset=<-1 0 0>}
     {limitRadius=8, drivingBone {boneName=thigh_twist_01_l}, sphereLocalOffset=<25 0 0>}
     {limitRadius=8, drivingBone {boneName=thigh_twist_01_l}, sphereLocalOffset=<5 0 0>}
    ]
   }
   n16 = Anim Dynamics - Bone: dyn_skirt_bk_01!(ComponentPose=@n15, Alpha=1, GravityOverride:Vector="0,0,-980") -> (Pose) {
    angularDampingOverride=0.999
    angularSpringConstant=500
    bAngularSpring=true
    bOverrideAngularDamping=true
    bUseGravityOverride=true
    simulationSpace=RootRelative
    boundBone {boneName=dyn_skirt_bk_01}
    componentAppliedLinearAccClamp=<3000 3000 3000>
    componentLinearAccScale=<0.5 0.5 0.5>
    componentLinearVelScale=<0.5 0.5 0.5>
    physicsBodyDefinitions [{
      collisionType=CustomSphere
      sphereCollisionRadius=2
      boundBone {boneName=dyn_skirt_bk_01}
      boxExtents=<25 25 25>
      constraintSetup {angularLimitsMax=<0 0 5>, angularLimitsMin=<0 -15 -5>, angularTarget=<1 0 0>}
      localJointOffset=<3 0 0>
     }]
    sphericalLimits [
     {limitRadius=7.5, drivingBone {boneName=thigh_twist_01_l}, sphereLocalOffset=<15 0 0>}
     {limitRadius=11, drivingBone {boneName=pelvis}, sphereLocalOffset=<-1 0 0>}
     {limitRadius=8, drivingBone {boneName=thigh_twist_01_l}, sphereLocalOffset=<25 0 0>}
     {limitRadius=8, drivingBone {boneName=thigh_twist_01_l}, sphereLocalOffset=<5 0 0>}
    ]
   }
   n17 = Anim Dynamics - Bone: dyn_skirt_bk_02!(ComponentPose=@n16, Alpha=1, GravityOverride:Vector="0,0,-980") -> (Pose) {
    angularDampingOverride=0.999
    angularSpringConstant=500
    bAngularSpring=true
    bOverrideAngularDamping=true
    bUseGravityOverride=true
    simulationSpace=RootRelative
    boundBone {boneName=dyn_skirt_bk_02}
    componentAppliedLinearAccClamp=<3000 3000 3000>
    componentLinearAccScale=<0.5 0.5 0.5>
    componentLinearVelScale=<0.5 0.5 0.5>
    physicsBodyDefinitions [{
      collisionType=CustomSphere
      sphereCollisionRadius=2
      boundBone {boneName=dyn_skirt_bk_02}
      boxExtents=<15 15 15>
      constraintSetup {angularLimitsMax=<0 0 5>, angularLimitsMin=<0 -15 -5>, angularTarget=<1 0 0>}
      localJointOffset=<3 0 0>
     }]
    sphericalLimits [
     {limitRadius=7.5, drivingBone {boneName=thigh_twist_01_l}, sphereLocalOffset=<15 0 0>}
     {limitRadius=11, drivingBone {boneName=pelvis}, sphereLocalOffset=<-1 0 0>}
     {limitRadius=8, drivingBone {boneName=thigh_twist_01_l}, sphereLocalOffset=<25 0 0>}
     {limitRadius=8, drivingBone {boneName=thigh_twist_01_l}, sphereLocalOffset=<5 0 0>}
    ]
   }
   Output Pose(Result=Component To Local(ComponentPose=@n17)) {layerGroup=None}
  ]}
 EventGraph {nodes [
   n1 = Event BlueprintUpdateAnimation()
   > Set Delta Time X(Delta Time X=@n1.DeltaTimeX)
   > Set is crouching(is crouching:double=SelectFloat(A=1, B=0, bPickA=Get bIsCrouching())) -> (Output_Get:double)
   n2 = Get is knee sliding()
   n3 = Get Delta Time X()
   n4 = Get Delta Time X()
   n5 = Get is knee sliding()
   > Set is knee sliding(is knee sliding:double=SelectFloat(
     A=FInterpTo(Current=@n2, Target=1, DeltaTime=@n3, InterpSpeed=5)
     B=FInterpTo(Current=@n5, Target=0, DeltaTime=@n4, InterpSpeed=5)
     bPickA=Get bIsKneeSliding()
    )) -> (Output_Get:double)
   n6 = Get Delta Time X()
   n7 = Get is baseball sliding()
   n8 = Get is baseball sliding()
   n9 = Get Delta Time X()
   > Set is baseball sliding(is baseball sliding:double=SelectFloat(
     A=FInterpTo(Current=@n7, Target=1, DeltaTime=@n6, InterpSpeed=10)
     B=FInterpTo(Current=@n8, Target=0, DeltaTime=@n9, InterpSpeed=5)
     bPickA=Get bIsBaseballSliding()
    )) -> (Output_Get:double)
  ]}
]
variables [
 Delta Time X {type=number}
 is baseball sliding {type=number}
 is crouching {type=number}
 is knee sliding {type=number}
]
