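# Control Rig for the ThunderBallet character's face accessory part, adding pose-driven correctives to her hair and ribbon bones.
# Clavicle pitch drives hair clavicle correctors, while head delta rotations are remapped into weight variables that offset and rotate the three back hair master bones.
# When the backpack flag is set, head yaw adds another back-hair offset.
# Head roll and spherical pose readers on right upper arm and head adjust the right front hair and left back ribbon.
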ControlRigBlueprint /BRCosmetics/Characters/Player/Female/Medium/Bodies/F_MED_ThunderBallet/Meshes/Parts/F_MED_ThunderBallet_FaceAcc_ControlRig
graphs [
 RigVMFunctionLibrary {links [], nodes []}
 RigVMFunction_Sequence_1_ContainedGraph {nodes [
   n1 = RigVMFunctionEntryNode()
   RigVMFunction_Sequence = &RigVMFunction_Sequence(ExecuteContext=@n1.ExecuteContext)
   n2 = &RigVMFunction_Sequence(ExecuteContext=@RigVMFunction_Sequence.B)
   n3 = &RigVMFunction_Sequence(ExecuteContext=@n2.B)
   n4 = &RigVMFunction_Sequence(ExecuteContext=@n3.B)
   n5 = &RigVMFunction_Sequence(ExecuteContext=@n4.B)
   RigVMFunctionReturnNode(A=@RigVMFunction_Sequence.A, B=@n2.A, C=@n3.A, D=@n4.A, E=@n5.A, F=@n5.B)
  ]}
 RigVMModel {nodes [
   RigUnit_BeginExecution = &RigUnit_BeginExecution()
   RigVMFunction_Sequence = RigVMAggregateNode(ExecuteContext=@RigUnit_BeginExecution.ExecutePin)
   Get Delta Ref = RigVMFunctionReferenceNode(Item=(Type=Bone,Name="clavicle_l"), ExecuteContext=@RigVMFunction_Sequence.A)
   > &RigUnit_ModifyTransforms(
    ItemToModify=((Item=(Type=Bone,Name="hair_clc_corr_L"),Transform=(Rotation=(X=0,Y=0,Z=0,W=1),Translation=(X=1,Y=3,Z=0),Scale3D=(X=1,Y=1,Z=1))))
    WeightMinimum=0
    Weight=&RigVMFunction_MathDoubleRemap(SourceMaximum=38, SourceMinimum=0, TargetMaximum=1, TargetMinimum=0, Value=@"Get Delta Ref".Result.Rotation.Pitch).Result
   )
   Get Delta Ref_2 = RigVMFunctionReferenceNode(Item=(Type=Bone,Name="clavicle_r"), ExecuteContext=@RigVMFunction_Sequence.B)
   > &RigUnit_ModifyTransforms(
    ItemToModify=((Item=(Type=Bone,Name="hair_clc_corr_R"),Transform=(Rotation=(X=0,Y=0,Z=0,W=1),Translation=(X=1,Y=-3,Z=0),Scale3D=(X=1,Y=1,Z=1))))
    WeightMinimum=0
    Weight=&RigVMFunction_MathDoubleRemap(SourceMaximum=38, SourceMinimum=0, TargetMaximum=1, TargetMinimum=0, Value=@"Get Delta Ref_2".Result.Rotation.Pitch).Result
   )
   Get Delta Ref_4 = RigVMFunctionReferenceNode(Item=(Type=Bone,Name="head"), ExecuteContext=@RigVMFunction_Sequence.C)
   > RigVMVariableNode(Value=&RigVMFunction_MathDoubleRemap(SourceMaximum=15, SourceMinimum=0, TargetMaximum=1, TargetMinimum=0, Value=@"Get Delta Ref_4".Result.Rotation.Pitch).Result) {variable=head rot y right}
   > RigVMVariableNode(Value=&RigVMFunction_MathDoubleRemap(SourceMaximum=-50, SourceMinimum=0, TargetMaximum=1, TargetMinimum=0, Value=@"Get Delta Ref_4".Result.Rotation.Yaw).Result) {variable=head rot z}
   > RigVMVariableNode(Value=&RigVMFunction_MathDoubleRemap(SourceMaximum=80, SourceMinimum=0, TargetMaximum=1, TargetMinimum=0, Value=@"Get Delta Ref_4".Result.Rotation.Roll).Result) {variable=head rot x left}
   > RigVMVariableNode(Value=&RigVMFunction_MathDoubleRemap(SourceMaximum=-80, SourceMinimum=0, TargetMaximum=1, TargetMinimum=0, Value=@"Get Delta Ref_4".Result.Rotation.Roll).Result) {variable=head rot x right}
   n1 = RigVMVariableNode() {variable=head rot z}
   > &RigUnit_ModifyTransforms(
    ItemToModify=((Item=(Type=Bone,Name="dyn_hair_bk_MSTR_L"),Transform=(Rotation=(X=0,Y=0,Z=0,W=1),Translation=(X=-5,Y=0,Z=5),Scale3D=(X=1,Y=1,Z=1))))
    WeightMinimum=0
    ItemToModify.0.Transform.Translation.Z=&RigVMFunction_MathDoubleRemap(SourceMaximum=1, SourceMinimum=0, TargetMaximum=-5, TargetMinimum=5, Value=RigVMVariableNode() {variable=head rot y right}.Value).Result
    Weight=@n1.Value
   )
   n2 = RigVMVariableNode() {variable=head rot y right}
   n3 = &RigVMFunction_MathDoubleRemap(SourceMaximum=1, SourceMinimum=0, TargetMaximum=-15, TargetMinimum=-5, Value=@n2.Value)
   > &RigUnit_ModifyTransforms(
    ItemToModify=((Item=(Type=Bone,Name="dyn_hair_bk_MSTR"),Transform=(Rotation=(X=0,Y=0,Z=0,W=1),Translation=(X=-5,Y=0,Z=5),Scale3D=(X=1,Y=1,Z=1))))
    WeightMinimum=0
    ItemToModify.0.Transform.Translation.X=@n3.Result
    Weight=@n1.Value
   )
   RigVMFunction_MathQuaternionFromEuler = &RigVMFunction_MathQuaternionFromEuler(
    Euler=(X=0,Y=0,Z=0)
    Euler.Z=&RigVMFunction_MathDoubleRemap(SourceMaximum=1, SourceMinimum=0, TargetMaximum=20, TargetMinimum=0, Value=@n2.Value).Result
   )
   > &RigUnit_ModifyTransforms(
    ItemToModify=((Item=(Type=Bone,Name="dyn_hair_bk_MSTR_R"),Transform=(Rotation=(X=0,Y=0,Z=0,W=1),Translation=(X=-5,Y=0,Z=5),Scale3D=(X=1,Y=1,Z=1))))
    WeightMinimum=0
    ItemToModify.0.Transform.Translation.X=@n3.Result
    ItemToModify.0.Transform.Rotation=@RigVMFunction_MathQuaternionFromEuler.Result
    Weight=@n1.Value
   )
   n4 = RigVMVariableNode() {variable=head rot x left}
   Remap = &RigVMFunction_MathDoubleRemap(SourceMaximum=1, SourceMinimum=0, TargetMaximum=-20, TargetMinimum=-10, Value=RigVMVariableNode() {variable=head rot z}.Value)
   > &RigUnit_ModifyTransforms(
    ItemToModify=((Item=(Type=Bone,Name="dyn_hair_bk_MSTR_L"),Transform=(Rotation=(X=0,Y=0,Z=0.1736,W=0.9848),Translation=(X=0,Y=0,Z=-10),Scale3D=(X=1,Y=1,Z=1))))
    WeightMinimum=0
    Weight=@n4.Value
    ItemToModify.0.Transform.Translation.Z=@Remap.Result
   )
   > &RigUnit_ModifyTransforms(
    ItemToModify=((Item=(Type=Bone,Name="dyn_hair_bk_MSTR"),Transform=(Rotation=(X=0,Y=0,Z=0.1736,W=0.9848),Translation=(X=0,Y=10,Z=0),Scale3D=(X=1,Y=1,Z=1))))
    WeightMinimum=0
    Weight=@n4.Value
   )
   > &RigUnit_ModifyTransforms(
    ItemToModify=((Item=(Type=Bone,Name="dyn_hair_bk_MSTR_R"),Transform=(Rotation=(X=0,Y=0,Z=0.1736,W=0.9848),Translation=(X=0,Y=10,Z=0),Scale3D=(X=1,Y=1,Z=1))))
    WeightMinimum=0
    Weight=@n4.Value
   )
   n5 = RigVMVariableNode() {variable=head rot x right}
   > &RigUnit_ModifyTransforms(
    ItemToModify=((Item=(Type=Bone,Name="dyn_hair_bk_MSTR_L"),Transform=(Rotation=(X=0,Y=0,Z=-0.1736,W=0.9848),Translation=(X=0,Y=-10,Z=0),Scale3D=(X=1,Y=1,Z=1))))
    WeightMinimum=0
    Weight=@n5.Value
   )
   > &RigUnit_ModifyTransforms(
    ItemToModify=((Item=(Type=Bone,Name="dyn_hair_bk_MSTR"),Transform=(Rotation=(X=0,Y=0,Z=-0.1736,W=0.9848),Translation=(X=0,Y=-10,Z=0),Scale3D=(X=1,Y=1,Z=1))))
    WeightMinimum=0
    Weight=@n5.Value
   )
   > &RigUnit_ModifyTransforms(
    ItemToModify=((Item=(Type=Bone,Name="dyn_hair_bk_MSTR_R"),Transform=(Rotation=(X=0,Y=0,Z=-0.1736,W=0.9848),Translation=(X=0,Y=0,Z=-10),Scale3D=(X=1,Y=1,Z=1))))
    WeightMinimum=0
    Weight=@n5.Value
    ItemToModify.0.Transform.Translation.Z=@Remap.Result
   )
   RigVMFunction_ControlFlowBranch = &RigVMFunction_ControlFlowBranch(ExecuteContext=@RigVMFunction_Sequence.D, Condition=RigVMVariableNode() {variable=is wearingbackpack}.Value)
   n6 = RigVMVariableNode() {variable=head rot z}
   &RigUnit_ModifyTransforms(
    ItemToModify=((Item=(Type=Bone,Name="dyn_hair_bk_MSTR_L"),Transform=(Rotation=(X=0,Y=0,Z=0,W=1),Translation=(X=5,Y=0,Z=0),Scale3D=(X=1,Y=1,Z=1))))
    WeightMinimum=0
    ExecutePin=@RigVMFunction_ControlFlowBranch.True
    Weight=@n6.Value
   )
   > &RigUnit_ModifyTransforms(
    ItemToModify=((Item=(Type=Bone,Name="dyn_hair_bk_MSTR"),Transform=(Rotation=(X=0,Y=0,Z=0,W=1),Translation=(X=5,Y=0,Z=0),Scale3D=(X=1,Y=1,Z=1))))
    WeightMinimum=0
    Weight=@n6.Value
   )
   > &RigUnit_ModifyTransforms(
    ItemToModify=((Item=(Type=Bone,Name="dyn_hair_bk_MSTR_R"),Transform=(Rotation=(X=0,Y=0,Z=0,W=1),Translation=(X=5,Y=0,Z=0),Scale3D=(X=1,Y=1,Z=1))))
    WeightMinimum=0
    Weight=@n6.Value
   )
   Get Delta Ref_3 = RigVMFunctionReferenceNode(Item=(Type=Bone,Name="head"), ExecuteContext=@RigVMFunction_Sequence.E)
   > &RigUnit_ModifyTransforms(
    ItemToModify=((Item=(Type=Bone,Name="dyn_hair_fr_MSTR_R"),Transform=(Rotation=(X=0,Y=-0.2588,Z=0,W=0.9659),Translation=(X=0,Y=0,Z=0),Scale3D=(X=1,Y=1,Z=1))))
    WeightMinimum=0
    Weight=&RigVMFunction_MathDoubleRemap(SourceMaximum=-40, SourceMinimum=0, TargetMaximum=1, TargetMinimum=0, Value=@"Get Delta Ref_3".Result.Rotation.Roll).Result
   )
   > n7 = &RigUnit_SphericalPoseReader(
    ActiveRegionScaleFactors=(PositiveWidth=1,NegativeWidth=1,PositiveHeight=1,NegativeHeight=1)
    ActiveRegionSize=0.15
    Debug=(bDrawDebug=false,DebugScale=25,DebugSegments=20,DebugThickness=0.25)
    DriverAxis=(X=-1,Y=0,Z=0)
    DriverItem=(Type=Bone,Name="upperarm_r")
    FalloffRegionScaleFactors=(PositiveWidth=1,NegativeWidth=1,PositiveHeight=1,NegativeHeight=1)
    FalloffSize=0.4
    FlipHeightScaling=false
    FlipWidthScaling=true
    OptionalParentItem=(Type=Bone)
    RotationOffset=(X=-120,Y=40,Z=0)
   )
   > &RigUnit_ModifyTransforms(
    ItemToModify=((Item=(Type=Bone,Name="dyn_hair_fr_MSTR_R"),Transform=(Rotation=(X=0.0594,Y=0.3368,Z=-0.1632,W=0.9254),Translation=(X=0,Y=0,Z=0),Scale3D=(X=1,Y=1,Z=1))))
    WeightMinimum=0
    Weight=@n7.OutputParam
   )
   n8 = &RigUnit_SphericalPoseReader(
    ActiveRegionScaleFactors=(PositiveWidth=1,NegativeWidth=1,PositiveHeight=1,NegativeHeight=1)
    ActiveRegionSize=0.07
    Debug=(bDrawDebug=false,DebugScale=25,DebugSegments=20,DebugThickness=0.25)
    DriverAxis=(X=0,Y=1,Z=0)
    DriverItem=(Type=Bone,Name="head")
    FalloffRegionScaleFactors=(PositiveWidth=1,NegativeWidth=1,PositiveHeight=1,NegativeHeight=1)
    FalloffSize=0.17
    FlipHeightScaling=false
    FlipWidthScaling=true
    OptionalParentItem=(Type=Bone)
    RotationOffset=(X=130,Y=0,Z=0)
    ExecutePin=@RigVMFunction_Sequence.F
   )
   > &RigUnit_ModifyTransforms(
    ItemToModify=((Item=(Type=Bone,Name="dyn_ribbon_bk_MSTR_L"),Transform=(Rotation=(X=-0.0114,Y=0.0864,Z=0.13,W=0.9877),Translation=(X=0,Y=0,Z=0),Scale3D=(X=1,Y=1,Z=1))))
    WeightMinimum=0
    Weight=@n8.OutputParam
   )
  ]}
]
hierarchy [
 root {loc=<0 0 0>, rot=<0 0 0 1>}
 pelvis {parent=root, loc=<0 -0.5301 98.0571>, rot=<0.0127 0.707 -0.0127 -0.707>}
 spine_01 {parent=pelvis, loc=<6.1208 -0.9328 0>, rot=<0 0 0.034 -0.9994>}
 spine_02 {parent=spine_01, loc=<6.5997 -0.2095 0>, rot=<0 0 0.0238 -0.9997>}
 spine_03 {parent=spine_02, loc=<6.4914 -0.1237 0>, rot=<0 0 0.0149 -0.9999>}
 spine_04 {parent=spine_03, loc=<6.9598 0.0794 0>, rot=<0 0 -0.0132 -0.9999>}
 spine_05 {parent=spine_04, loc=<6.9987 0 0>, rot=<0 0 -0.0015 -1>}
 neck_01 {parent=spine_05, loc=<11.3138 0.1529 0>, rot=<0 0 -0.2003 -0.9797>}
 neck_02 {parent=neck_01, loc=<4.5627 0 0>, rot=<0 0 0.0504 -0.9987>}
 head {parent=neck_02, loc=<5.3914 0 0>, rot=<0 0 0.115 -0.9934>}
 dyn_hair_fr_MSTR_R {parent=head, loc=<0.4208 6.5022 7.3432>, rot=<0.1349 0.1205 0.9624 -0.2025>}
 dyn_hair_fr_01_R {parent=dyn_hair_fr_MSTR_R, loc=<0 0 0>, rot=<0 0 0 1>}
 dyn_hair_fr_02_R {parent=dyn_hair_fr_01_R, loc=<8 0 0>, rot=<0 0 0.3973 -0.9177>}
 dyn_ribbon_bk_MSTR_R {parent=head, loc=<-1.2496 3.3492 6.4729>, rot=<-0.1218 0.0024 0.9887 -0.0868>}
 dyn_ribbon_bk_01_R {parent=dyn_ribbon_bk_MSTR_R, loc=<0 0 0>, rot=<0 0 0 1>}
 dyn_ribbon_bk_02_R {parent=dyn_ribbon_bk_01_R, loc=<7.5 0 0>, rot=<0 0.1329 0 -0.9911>}
 dyn_ribbon_bk_MSTR_L {parent=head, loc=<-1.1914 2.7614 -7.6196>, rot=<-0.1268 -0.9622 -0.0922 0.2225>}
 dyn_ribbon_bk_01_L {parent=dyn_ribbon_bk_MSTR_L, loc=<0 0 0>, rot=<0 0 0 1>}
 dyn_ribbon_bk_02_L {parent=dyn_ribbon_bk_01_L, loc=<7.5 0 0>, rot=<0.0397 0.2025 0.1882 -0.9602>}
 dyn_ribbon_fr_MSTR_L {parent=head, loc=<9.8397 10.9259 -2.0674>, rot=<-0.0331 -0.956 -0.1953 -0.2162>}
 dyn_ribbon_fr_01_L {parent=dyn_ribbon_fr_MSTR_L, loc=<0 0 0>, rot=<0 0 0 1>}
 dyn_ribbon_fr_02_L {parent=dyn_ribbon_fr_01_L, loc=<7.5 0 0>, rot=<0 -0.2203 0 -0.9754>}
 dyn_hair_fr_MSTR_L {parent=head, loc=<10.2035 11.1823 -2.7553>, rot=<0.0514 -0.9778 -0.0654 -0.1925>}
 dyn_hair_fr_01_L {parent=dyn_hair_fr_MSTR_L, loc=<0 0 0>, rot=<0 0 0 1>}
 dyn_hair_fr_02_L {parent=dyn_hair_fr_01_L, loc=<9.3 0 0>, rot=<0.0117 -0.1138 0.1015 -0.9882>}
 dyn_hair_bk_MSTR {parent=head, loc=<9.2149 -6.5062 0>, rot=<0.0779 -0.7028 0.7028 -0.0779>}
 dyn_hair_bk_01 {parent=dyn_hair_bk_MSTR, loc=<0 0 0>, rot=<0 0 0 1>}
 dyn_hair_bk_02 {parent=dyn_hair_bk_01, loc=<10 0 0>, rot=<0 0 0 1>}
 dyn_hair_bk_03 {parent=dyn_hair_bk_02, loc=<10 0 0>, rot=<0 0.1219 0 -0.9925>}
 dyn_hair_bk_04 {parent=dyn_hair_bk_03, loc=<10.2814 0 0>, rot=<0 -0.1822 0 -0.9833>}
 dyn_hair_bk_05 {parent=dyn_hair_bk_04, loc=<10 0 0>, rot=<0 -0.0422 0 -0.9991>}
 dyn_hair_bk_06 {parent=dyn_hair_bk_05, loc=<10 0 0>, rot=<0 0 0 1>}
 dyn_hair_bk_MSTR_L {parent=head, loc=<8.876 -3.532 -7.258>, rot=<0.0936 -0.7444 0.6498 -0.1223>}
 dyn_hair_bk_01_L {parent=dyn_hair_bk_MSTR_L, loc=<0 0 0>, rot=<0 0 0 1>}
 dyn_hair_bk_02_L {parent=dyn_hair_bk_01_L, loc=<10.0001 0 -0.0001>, rot=<0 0 0 1>}
 dyn_hair_bk_03_L {parent=dyn_hair_bk_02_L, loc=<9.9999 0 0>, rot=<0 0.0872 0 -0.9962>}
 dyn_hair_bk_04_L {parent=dyn_hair_bk_03_L, loc=<10 0 0>, rot=<0 -0.2164 0 -0.9763>}
 dyn_hair_bk_05_L {parent=dyn_hair_bk_04_L, loc=<10 0 0>, rot=<0 0 0 1>}
 dyn_hair_bk_06_L {parent=dyn_hair_bk_05_L, loc=<10 0 0>, rot=<0 0 0 1>}
 hair_bk_corr {parent=head, loc=<0 0 0>, rot=<0 0 0 1>}
 dyn_hair_bk_MSTR_R {parent=head, loc=<8.8761 -3.5317 7.2579>, rot=<0.1223 -0.6498 0.7444 -0.0936>}
 dyn_hair_bk_01_R {parent=dyn_hair_bk_MSTR_R, loc=<0 0 0>, rot=<0 0 0 -1>}
 dyn_hair_bk_02_R {parent=dyn_hair_bk_01_R, loc=<10.0001 0 -0.0001>, rot=<0 0 0 1>}
 dyn_hair_bk_03_R {parent=dyn_hair_bk_02_R, loc=<10.0001 0 -0.0001>, rot=<0 0.0872 0 -0.9962>}
 dyn_hair_bk_04_R {parent=dyn_hair_bk_03_R, loc=<10 -0.0001 0>, rot=<0 -0.2164 0 -0.9763>}
 dyn_hair_bk_05_R {parent=dyn_hair_bk_04_R, loc=<10 0 -0.0001>, rot=<0 0 0 -1>}
 dyn_hair_bk_06_R {parent=dyn_hair_bk_05_R, loc=<10 0 0>, rot=<0 0 0 -1>}
 FACIAL_C_FacialRoot {parent=head, loc=<2.1524 -0.746 -0.0002>, rot=<0.4879 -0.5118 -0.4879 -0.5118>}
 FACIAL_C_Forehead {parent=FACIAL_C_FacialRoot, loc=<0.0009 -2.8304 9.4585>, rot=<-0.041 0 0 -0.9992>}
 FACIAL_L_ForeheadIn {parent=FACIAL_C_FacialRoot, loc=<1.6959 -2.7425 9.3246>, rot=<-0.0583 -0.0696 0.0045 -0.9959>}
 FACIAL_R_ForeheadIn {parent=FACIAL_C_FacialRoot, loc=<-1.6973 -2.7408 9.3233>, rot=<-0.0583 0.0696 -0.0044 -0.9959>}
 FACIAL_L_ForeheadMid {parent=FACIAL_C_FacialRoot, loc=<3.1582 -2.9948 8.7931>, rot=<-0.0686 -0.1757 -0.0202 -0.9818>}
 FACIAL_R_ForeheadMid {parent=FACIAL_C_FacialRoot, loc=<-3.161 -2.9914 8.7905>, rot=<-0.0686 0.1757 0.0202 -0.9818>}
 FACIAL_L_ForeheadOut {parent=FACIAL_C_FacialRoot, loc=<4.5384 -2.9312 7.8305>, rot=<-0.1191 -0.2558 -0.0785 -0.9562>}
 FACIAL_R_ForeheadOut {parent=FACIAL_C_FacialRoot, loc=<-4.5398 -2.9297 7.8293>, rot=<-0.119 0.2558 0.0786 -0.9561>}
 FACIAL_L_EyesackUpper {parent=FACIAL_C_FacialRoot, loc=<3.1579 -1.6966 9.2412>, rot=<0.0512 -0.0844 -0.0643 -0.993>}
 FACIAL_R_EyesackUpper {parent=FACIAL_C_FacialRoot, loc=<-3.1619 -1.6945 9.2404>, rot=<0.0512 0.0844 0.0643 -0.993>}
 FACIAL_L_EyelidUpperFurrow {parent=FACIAL_C_FacialRoot, loc=<2.937 -1.3998 9.0362>, rot=<-0.0277 -0.151 -0.0793 -0.985>}
 FACIAL_R_EyelidUpperFurrow {parent=FACIAL_C_FacialRoot, loc=<-2.9411 -1.3981 9.0349>, rot=<-0.0277 0.151 0.0793 -0.985>}
 FACIAL_L_EyelidUpperA {parent=FACIAL_C_FacialRoot, loc=<3.011 -0.6226 7.2295>, rot=<-0.1454 0 0 -0.9894>}
 FACIAL_R_EyelidUpperA {parent=FACIAL_C_FacialRoot, loc=<-3.0113 -0.623 7.2296>, rot=<-0.1454 0 0 -0.9894>}
 clavicle_l {parent=spine_05, loc=<5.4831 -0.1053 -3.2946>, rot=<0.0465 -0.7289 -0.009 -0.6829>}
 upperarm_l {parent=clavicle_l, loc=<9.8658 -1.2136 -0.4105>, rot=<0.0164 -0.3245 -0.0269 -0.9454>}
 upperarm_twist_01_l {parent=upperarm_l, loc=<7.2135 -0.1134 0>, rot=<0 0 0 1>}
 hair_clc_corr_L {parent=clavicle_l, loc=<6.4641 -1.4067 5.5761>, rot=<-0.0241 0.0388 0.6369 -0.7696>}
 clavicle_r {parent=spine_05, loc=<5.4831 -0.1053 3.2946>, rot=<0.7289 0.0465 -0.6829 0.009>}
 upperarm_r {parent=clavicle_r, loc=<-9.8658 1.2136 0.4105>, rot=<0.0164 -0.3245 -0.0269 -0.9454>}
 upperarm_twist_01_r {parent=upperarm_r, loc=<-7.2135 0.1134 0>, rot=<0 0 0 -1>}
 hair_clc_corr_R {parent=clavicle_r, loc=<-6.4641 1.4067 -5.5762>, rot=<-0.6369 -0.7696 -0.0241 -0.0388>}
]
variables [
 head rot x left {type=double, default=0}
 head rot x right {type=double, default=0}
 head rot y right {type=double, default=0}
 head rot z {type=double, default=0}
 is wearingbackpack {type=bool, default=False}
]
