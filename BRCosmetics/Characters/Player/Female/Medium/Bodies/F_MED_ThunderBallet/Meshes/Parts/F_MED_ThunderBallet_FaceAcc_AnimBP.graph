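# Post-process style Anim Blueprint for the ThunderBallet face accessory, adding secondary motion to hair and ribbons.
# Copied parent pose, with upper arms layered from the body mesh, feeds a Control Rig and cached pose.
# A backpack/cape flag selects between two RigidBody simulations, the backpack one adding back-hair bone offsets.
# AnimDynamics springs then move front hair and ribbons, some with spherical body limits, while the event graph updates state and gravity settings.
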AnimBlueprint /BRCosmetics/Characters/Player/Female/Medium/Bodies/F_MED_ThunderBallet/Meshes/Parts/F_MED_ThunderBallet_FaceAcc_AnimBP
target_skeleton=&/F_MED_ThunderBallet_FaceAcc_Skeleton
graphs [
 AnimGraph {nodes [
   n1 = RigidBody!(
    ComponentPose=Local To Component(LocalPose=Use cached pose 'start pose'())
    Alpha
    OverrideWorldGravity:Vector="0,0,-980"
    SimSpaceSettings:Sim Space Settings Structure=(WorldAlpha=1,VelocityScaleZ=1,DampingAlpha=1,MaxLinearVelocity=10000,MaxAngularVelocity=10000,MaxLinearAcceleration=10000,MaxAngularAcceleration=10000,ExternalLinearDragV=(X=0,Y=0,Z=0),ExternalLinearVelocity=(X=0,Y=0,Z=0),ExternalAngularVelocity=(X=0,Y=0,Z=0))
   ) -> (Pose) {
    bClampLinearTranslationLimitToRefPose=true
    bOverrideWorldGravity=true
    overridePhysicsAsset=&/F_MED_ThunderBallet_FaceAcc_BackPack_Physics
    simulationSpace=BaseBoneSpace
    baseBoneRef {boneName=neck_01}
    simSpaceSettings {worldAlpha=1}
   }
   n2 = "Transform (Modify) Bone - Bone: dyn_hair_bk_04_L"(ComponentPose=@n1, Alpha=1, Translation="0,0,0", Rotation="-10,0,0", Scale="1,1,0.8") {
    rotationMode=BMM_Additive
    rotationSpace=BCS_BoneSpace
    scaleMode=BMM_Replace
    scaleSpace=BCS_BoneSpace
    boneToModify {boneName=dyn_hair_bk_04_L}
    rotation {pitch=-10}
    scale {z=0.8}
   }
   n3 = "Transform (Modify) Bone - Bone: dyn_hair_bk_04"(ComponentPose=@n2, Alpha=1, Translation="0,0,0", Rotation="-10,0,0", Scale="1,1,0.8") {
    rotationMode=BMM_Additive
    rotationSpace=BCS_BoneSpace
    scaleMode=BMM_Replace
    scaleSpace=BCS_BoneSpace
    boneToModify {boneName=dyn_hair_bk_04}
    rotation {pitch=-10}
    scale {z=0.8}
   }
   n4 = "Transform (Modify) Bone - Bone: dyn_hair_bk_04_R"(ComponentPose=@n3, Alpha=1, Translation="0,0,0", Rotation="-10,0,0", Scale="1,1,0.8") {
    rotationMode=BMM_Additive
    rotationSpace=BCS_BoneSpace
    scaleMode=BMM_Replace
    scaleSpace=BCS_BoneSpace
    boneToModify {boneName=dyn_hair_bk_04_R}
    rotation {pitch=-10}
    scale {z=0.8}
   }
   n5 = "Transform (Modify) Bone - Bone: dyn_hair_bk_05_L"(ComponentPose=@n4, Alpha=1, Translation="0,0,0", Rotation="12,0,0", Scale="1,1,1") {rotationMode=BMM_Additive, rotationSpace=BCS_BoneSpace, boneToModify {boneName=dyn_hair_bk_05_L}, rotation {pitch=12}}
   n6 = "Transform (Modify) Bone - Bone: dyn_hair_bk_05_R"(
    ComponentPose="Transform (Modify) Bone - Bone: dyn_hair_bk_05"(ComponentPose=@n5, Alpha=1, Translation="0,0,0", Rotation="12,0,0", Scale="1,1,1") {rotationMode=BMM_Additive, rotationSpace=BCS_BoneSpace, boneToModify {boneName=dyn_hair_bk_05}, rotation {pitch=12}}
    Alpha=1
    Translation="0,0,0"
    Rotation="12,0,0"
    Scale="1,1,1"
   ) {rotationMode=BMM_Additive, rotationSpace=BCS_BoneSpace, boneToModify {boneName=dyn_hair_bk_05_R}, rotation {pitch=12}}
   n7 = Component To Local(ComponentPose=@n6)
   n8 = RigidBody(
    ComponentPose=Local To Component(LocalPose=Use cached pose 'start pose'())
    OverrideWorldGravity="0,0,-980"
    SimSpaceSettings=(WorldAlpha=1,VelocityScaleZ=1,DampingAlpha=1,MaxLinearVelocity=10000,MaxAngularVelocity=10000,MaxLinearAcceleration=10000,MaxAngularAcceleration=10000,ExternalLinearDragV=(X=0,Y=0,Z=0),ExternalLinearVelocity=(X=0,Y=0,Z=0),ExternalAngularVelocity=(X=0,Y=0,Z=0))
   ) {
    bClampLinearTranslationLimitToRefPose=true
    bOverrideWorldGravity=true
    overridePhysicsAsset=&/F_MED_ThunderBallet_FaceAcc_Physics
    simulationSpace=BaseBoneSpace
    baseBoneRef {boneName=neck_01}
    simSpaceSettings {worldAlpha=1}
   }
   n9 = Component To Local(ComponentPose=@n8)
   n10 = Get isWearingBackpack/Cape() -> (isWearingBackpack/Cape:bool)
   n11 = Anim Dynamics - Bone: dyn_hair_fr_01_L!(
    ComponentPose=Local To Component(LocalPose=Blend Poses by bool(BlendPose_0=@n7, BlendPose_1=@n9, bActiveValue=@n10))
    Alpha=1
    GravityOverride:Vector="0,0,-980"
   ) -> (Pose) {
    angularDampingOverride=0.99
    angularSpringConstant=500
    bAngularSpring=true
    bOverrideAngularDamping=true
    bUseGravityOverride=true
    simulationSpace=RootRelative
    boundBone {boneName=dyn_hair_fr_01_L}
    componentAppliedLinearAccClamp=<3000 3000 3000>
    componentLinearAccScale=<0.3 0.3 0.3>
    componentLinearVelScale=<0.3 0.3 0.3>
    physicsBodyDefinitions [{
      boundBone {boneName=dyn_hair_fr_01_L}
      boxExtents=<25 25 25>
      constraintSetup {angularLimitsMax=<0 0 45>, angularLimitsMin=<0 -5 0>, angularTarget=<1 0 0>}
      localJointOffset=<2 0 0>
     }]
   }
   n12 = Anim Dynamics - Bone: dyn_hair_fr_02_L!(ComponentPose=@n11, Alpha=1, GravityOverride:Vector="0,0,-980") -> (Pose) {
    angularDampingOverride=0.99
    angularSpringConstant=500
    bAngularSpring=true
    bOverrideAngularDamping=true
    bUseGravityOverride=true
    simulationSpace=RootRelative
    boundBone {boneName=dyn_hair_fr_02_L}
    componentAppliedLinearAccClamp=<3000 3000 3000>
    componentLinearAccScale=<0.3 0.3 0.3>
    componentLinearVelScale=<0.3 0.3 0.3>
    physicsBodyDefinitions [{
      boundBone {boneName=dyn_hair_fr_02_L}
      boxExtents=<15 15 15>
      constraintSetup {angularLimitsMax=<0 0 5>, angularLimitsMin=<0 -5 0>, angularTarget=<1 0 0>}
      localJointOffset=<1 0 0>
     }]
   }
   n13 = Anim Dynamics - Bone: dyn_hair_fr_01_R!(ComponentPose=@n12, Alpha=1, GravityOverride:Vector="0,0,-980") -> (Pose) {
    angularDampingOverride=0.99
    angularSpringConstant=1000
    bAngularSpring=true
    bOverrideAngularDamping=true
    bUseGravityOverride=true
    simulationSpace=RootRelative
    boundBone {boneName=dyn_hair_fr_01_R}
    componentAppliedLinearAccClamp=<3000 3000 3000>
    componentLinearAccScale=<0.3 0.3 0.3>
    componentLinearVelScale=<0.3 0.3 0.3>
    physicsBodyDefinitions [{
      boundBone {boneName=dyn_hair_fr_01_R}
      boxExtents=<25 25 25>
      constraintSetup {angularLimitsMax=<0 0 10>, angularLimitsMin=<0 -15 -35>, angularTarget=<1 0 0>}
      localJointOffset=<1 0 0>
     }]
   }
   n14 = Anim Dynamics - Bone: dyn_hair_fr_02_R!(ComponentPose=@n13, Alpha=1, GravityOverride:Vector="0,0,-980") -> (Pose) {
    angularDampingOverride=0.99
    angularSpringConstant=1000
    bAngularSpring=true
    bOverrideAngularDamping=true
    bUseGravityOverride=true
    bUseSphericalLimits=true
    simulationSpace=RootRelative
    boundBone {boneName=dyn_hair_fr_02_R}
    componentAppliedLinearAccClamp=<3000 3000 3000>
    componentLinearAccScale=<0.3 0.3 0.3>
    componentLinearVelScale=<0.3 0.3 0.3>
    physicsBodyDefinitions [{
      collisionType=CustomSphere
      sphereCollisionRadius=3
      boundBone {boneName=dyn_hair_fr_02_R}
      boxExtents=<15 15 15>
      constraintSetup {angularLimitsMax=<0 5 0>, angularLimitsMin=<0 -15 -15>, angularTarget=<1 0 0>}
      localJointOffset=<4.5 0 0>
     }]
    sphericalLimits [
     {limitRadius=9.5, drivingBone {boneName=spine_05}, sphereLocalOffset=<-1 -2 0>}
     {limitRadius=8.7, drivingBone {boneName=clavicle_l}, sphereLocalOffset=<4 0 -5.5>}
     {limitRadius=8.7, drivingBone {boneName=clavicle_r}, sphereLocalOffset=<-4 0 5.5>}
     {limitRadius=9.7, drivingBone {boneName=spine_05}, sphereLocalOffset=<-5 2 0>}
     {limitRadius=9.7, drivingBone {boneName=spine_05}, sphereLocalOffset=<-5 2 -5>}
     {limitRadius=9.7, drivingBone {boneName=spine_05}, sphereLocalOffset=<-5 2 5>}
     {limitRadius=4.8, drivingBone {boneName=neck_01}, sphereLocalOffset=<-1 0 0>}
    ]
   }
   n15 = Anim Dynamics - Bone: dyn_ribbon_fr_01_L!(ComponentPose=@n14, Alpha=1, GravityOverride:Vector="0,0,-980") -> (Pose) {
    angularDampingOverride=0.999
    angularSpringConstant=100
    bAngularSpring=true
    bOverrideAngularDamping=true
    bUseGravityOverride=true
    simulationSpace=RootRelative
    boundBone {boneName=dyn_ribbon_fr_01_L}
    componentAppliedLinearAccClamp=<3000 3000 3000>
    componentLinearAccScale=<0.3 0.3 0.3>
    componentLinearVelScale=<0.3 0.3 0.3>
    physicsBodyDefinitions [{
      boundBone {boneName=dyn_ribbon_fr_01_L}
      boxExtents=<20 20 20>
      constraintSetup {angularLimitsMax=<0 0 35>, angularLimitsMin=<0 -5 0>, angularTarget=<1 0 0>}
      localJointOffset=<1 0 0>
     }]
   }
   n16 = Anim Dynamics - Bone: dyn_ribbon_fr_02_L!(ComponentPose=@n15, Alpha=1, GravityOverride:Vector="0,0,-980") -> (Pose) {
    angularDampingOverride=0.999
    angularSpringConstant=100
    bAngularSpring=true
    bOverrideAngularDamping=true
    bUseGravityOverride=true
    simulationSpace=RootRelative
    boundBone {boneName=dyn_ribbon_fr_02_L}
    componentAppliedLinearAccClamp=<3000 3000 3000>
    componentLinearAccScale=<0.3 0.3 0.3>
    componentLinearVelScale=<0.3 0.3 0.3>
    physicsBodyDefinitions [{
      boundBone {boneName=dyn_ribbon_fr_02_L}
      boxExtents=<10 10 10>
      constraintSetup {angularLimitsMax=<0 0 5>, angularLimitsMin=<0 -5 0>, angularTarget=<1 0 0>}
      localJointOffset=<1 0 0>
     }]
   }
   n17 = Anim Dynamics - Bone: dyn_ribbon_bk_01_L!(ComponentPose=@n16, Alpha=1, GravityOverride:Vector="0,0,-980") -> (Pose) {
    angularDampingOverride=0.999
    angularSpringConstant=100
    bAngularSpring=true
    bOverrideAngularDamping=true
    bUseGravityOverride=true
    simulationSpace=RootRelative
    boundBone {boneName=dyn_ribbon_bk_01_L}
    componentAppliedLinearAccClamp=<3000 3000 3000>
    componentLinearAccScale=<0.3 0.3 0.3>
    componentLinearVelScale=<0.3 0.3 0.3>
    physicsBodyDefinitions [{
      boundBone {boneName=dyn_ribbon_bk_01_L}
      boxExtents=<20 20 20>
      constraintSetup {angularLimitsMax=<0 10 20>, angularLimitsMin=<0 0 0>, angularTarget=<1 0 0>}
      localJointOffset=<1 0 0>
     }]
   }
   n18 = Anim Dynamics - Bone: dyn_ribbon_bk_02_L!(ComponentPose=@n17, Alpha=1, GravityOverride:Vector="0,0,-980") -> (Pose) {
    angularDampingOverride=0.999
    angularSpringConstant=100
    bAngularSpring=true
    bOverrideAngularDamping=true
    bUseGravityOverride=true
    bUseSphericalLimits=true
    simulationSpace=RootRelative
    boundBone {boneName=dyn_ribbon_bk_02_L}
    componentAppliedLinearAccClamp=<3000 3000 3000>
    componentLinearAccScale=<0.3 0.3 0.3>
    componentLinearVelScale=<0.3 0.3 0.3>
    physicsBodyDefinitions [{
      collisionType=CustomSphere
      sphereCollisionRadius=2.5
      boundBone {boneName=dyn_ribbon_bk_02_L}
      boxExtents=<10 10 10>
      constraintSetup {angularLimitsMax=<0 10 20>, angularLimitsMin=<0 -10 -5>, angularTarget=<1 0 0>}
      localJointOffset=<4 0 0>
     }]
    sphericalLimits [
     {limitRadius=10, drivingBone {boneName=spine_05}, sphereLocalOffset=<-1 -2 0>}
     {limitRadius=9, drivingBone {boneName=clavicle_l}, sphereLocalOffset=<4 0 -5.5>}
     {limitRadius=9, drivingBone {boneName=clavicle_r}, sphereLocalOffset=<-4 0 5.5>}
     {limitRadius=10, drivingBone {boneName=spine_05}, sphereLocalOffset=<-5 2 0>}
     {limitRadius=10, drivingBone {boneName=spine_05}, sphereLocalOffset=<-5 2 -5>}
     {limitRadius=10, drivingBone {boneName=spine_05}, sphereLocalOffset=<-5 2 5>}
     {limitRadius=5, drivingBone {boneName=neck_01}, sphereLocalOffset=<-1 0 0>}
    ]
   }
   n19 = Anim Dynamics - Bone: dyn_ribbon_bk_01_R!(ComponentPose=@n18, Alpha=1, GravityOverride:Vector="0,0,-980") -> (Pose) {
    angularDampingOverride=0.999
    angularSpringConstant=100
    bAngularSpring=true
    bOverrideAngularDamping=true
    bUseGravityOverride=true
    simulationSpace=RootRelative
    boundBone {boneName=dyn_ribbon_bk_01_R}
    componentAppliedLinearAccClamp=<3000 3000 3000>
    componentLinearAccScale=<0.3 0.3 0.3>
    componentLinearVelScale=<0.3 0.3 0.3>
    physicsBodyDefinitions [{
      boundBone {boneName=dyn_ribbon_bk_01_R}
      boxExtents=<20 20 20>
      constraintSetup {angularLimitsMax=<0 0 0>, angularLimitsMin=<0 -10 -20>, angularTarget=<1 0 0>}
      localJointOffset=<1 0 0>
     }]
   }
   n20 = Anim Dynamics - Bone: dyn_ribbon_bk_02_R!(ComponentPose=@n19, Alpha=1, GravityOverride:Vector="0,0,-980") -> (Pose) {
    angularDampingOverride=0.999
    angularSpringConstant=100
    bAngularSpring=true
    bOverrideAngularDamping=true
    bUseGravityOverride=true
    bUseSphericalLimits=true
    simulationSpace=RootRelative
    boundBone {boneName=dyn_ribbon_bk_02_R}
    componentAppliedLinearAccClamp=<3000 3000 3000>
    componentLinearAccScale=<0.3 0.3 0.3>
    componentLinearVelScale=<0.3 0.3 0.3>
    physicsBodyDefinitions [{
      collisionType=CustomSphere
      sphereCollisionRadius=2
      boundBone {boneName=dyn_ribbon_bk_02_R}
      boxExtents=<10 10 10>
      constraintSetup {angularLimitsMax=<0 10 20>, angularLimitsMin=<0 -10 -20>, angularTarget=<1 0 0>}
      localJointOffset=<2 0 0>
     }]
    sphericalLimits [
     {limitRadius=9.5, drivingBone {boneName=spine_05}, sphereLocalOffset=<-1 -2 0>}
     {limitRadius=8.5, drivingBone {boneName=clavicle_l}, sphereLocalOffset=<4 0 -5.5>}
     {limitRadius=8.5, drivingBone {boneName=clavicle_r}, sphereLocalOffset=<-4 0 5.5>}
     {limitRadius=9.5, drivingBone {boneName=spine_05}, sphereLocalOffset=<-5 2 0>}
     {limitRadius=9.5, drivingBone {boneName=spine_05}, sphereLocalOffset=<-5 2 -5>}
     {limitRadius=9.5, drivingBone {boneName=spine_05}, sphereLocalOffset=<-5 2 5>}
     {limitRadius=4.7, drivingBone {boneName=neck_01}, sphereLocalOffset=<-1 0 0>}
    ]
   }
   Output Pose(Result=Component To Local(ComponentPose=@n20)) {layerGroup=None}
   n21 = Get Body() -> (Body:Skeletal Mesh Component Budgeted Object Reference)
   n22 = Control Rig!(
    Source=Layered blend per bone(
     BasePose=Copy Pose From Mesh() {bUseAttachedParent=true}
     BlendPoses_0=Copy Pose From Mesh(SourceMeshComponent=@n21) {bUseAttachedParent=true}
    ) {layerSetup [{branchFilters [{blendDepth=2, boneName=upperarm_l}, {blendDepth=2, boneName=upperarm_r}]}]}
    Alpha
    is wearingbackpack:bool=Get isWearingBackpack/Cape()
   ) -> (Pose) {controlRigAssetReference {blueprintRigClass=&/F_MED_ThunderBallet_FaceAcc_ControlRig.F_MED_ThunderBallet_FaceAcc_ControlRig_C}}
   Save cached pose 'start pose'(Pose=@n22) {cacheName=start pose}
  ]}
 EventGraph {nodes [
   Event BlueprintUpdateAnimation()
   n1 = CheckForBackbling()
   > Set isWearingBackpack/Cape(isWearingBackpack/Cape:bool=OR Boolean(A=@n1, B=@n1.bIsBackblingACape)) -> (Output_Get:bool)
   > Set isBeginPlay(isBeginPlay:double=SelectFloat(A=1, B=Get isBeginPlay() -> (isBeginPlay:double), bPickA=Get bIsInFrontEnd())) -> (Output_Get:double)
   > Set is skydiving(is skydiving:double=SelectFloat(A=1, B=0, bPickA=Get bIsSkyDiving())) -> (Output_Get:double)
   > n2 = "Gravity Override State Function (Parameter Asset) (Nativized)"(
    CharacterPartAnimInstance=Self-Reference()
    GameTimeInSeconds=GetGameTimeInSeconds()
    Parameters=&/F_MED_ThunderBallet_FaceAcc_Parameters
   )
   > Set Gravity Override(Gravity Override=@n2)
   > Set RBAN Settings Out(RBAN Settings Out=@n2.OutRigidBodyAnimNodeParameters)
   n3 = Event OnCharacterCustomizationLoaded()
   > Set Body(Body:Skeletal Mesh Component Budgeted Object Reference=Get Array Item with EFort Custom Part Type(Array=@n3.CustomMeshParts)) -> (Output_Get:Skeletal Mesh Component Budgeted Object Reference)
   Event BlueprintBeginPlay() {is=title K2Node_Event}
   > Set isBeginPlay(isBeginPlay=1)
   Event BlueprintInitializeAnimation()
   > Set isBeginPlay(isBeginPlay=0)
  ]}
]
variables [
 Body {type=SkeletalMeshComponentBudgeted}
 Gravity Override {type=Vector}
 is skydiving {type=number}
 isBeginPlay {type=number}
 isWearingBackpack/Cape {type=boolean}
 RBAN Settings Out {type=FortRigidBodyAnimNodeParameters}
]
